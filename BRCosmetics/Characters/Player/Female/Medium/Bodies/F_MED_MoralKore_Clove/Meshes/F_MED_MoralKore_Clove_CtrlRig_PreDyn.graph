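# Pre-dynamics Control Rig for the Clove MoralKore body, correcting jacket, belt and collar bones from body pose.
# It runs a shared DeformRig function, then jacket and belt corrections using parent constraints and thigh spherical pose readers, partly crouch-dependent.
# Collar bones get weighted spine, neck and clavicle constraints, then rotate by dot product against a head vector.
# Clavicle and right upper-arm pose readers add interpolated collar offsets gated by `target`.
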ControlRigBlueprint /BRCosmetics/Characters/Player/Female/Medium/Bodies/F_MED_MoralKore_Clove/Meshes/F_MED_MoralKore_Clove_CtrlRig_PreDyn
graphs [
 CollapseNode_ContainedGraph {nodes [
   RigVMFunctionEntryNode()
   Multiply = RigVMAggregateNode(
    C=&RigUnit_GetTransform(Item=(Type=Bone,Name="neck_01"), Space=LocalSpace, bInitial=False).Transform
    A=&RigUnit_GetTransform(Item=(Type=Bone,Name="head_end"), Space=LocalSpace, bInitial=False).Transform
    D=&RigUnit_GetTransform(Item=(Type=Bone,Name="neck_02"), Space=LocalSpace, bInitial=False).Transform
    B=&RigUnit_GetTransform(Item=(Type=Bone,Name="head"), Space=LocalSpace, bInitial=False).Transform
   )
   Unit = /Script/RigVM.RigVMFunction_MathVectorUnit(Value=@Multiply.Result.Translation) {is=struct RigVMUnitNode, template="Unit::Execute(in Value,out Result)"}
   RigVMFunctionReturnNode(Result=@Unit.Result)
  ]}
 Multiply_1_ContainedGraph {nodes [
   n1 = RigVMFunctionEntryNode()
   Multiply = &RigVMFunction_MathTransformMul(A=@n1.A, B=@n1.B) {template="Multiply::Execute(in A,in B,out Result)"}
   RigVMFunctionReturnNode(Result=&RigVMFunction_MathTransformMul(
     B=@n1.D
     A=&RigVMFunction_MathTransformMul(B=@n1.C, A=@Multiply.Result) {template="Multiply::Execute(in A,in B,out Result)"}.Result
    ) {template="Multiply::Execute(in A,in B,out Result)"}.Result)
  ]}
 New Function_ContainedGraph {nodes [
   n1 = RigVMFunctionEntryNode()
   Dot__ = /Script/RigVM.RigVMFunction_MathVectorDot(
    B=&RigVMFunction_MathVectorUnit(Value=&RigUnit_GetTransform(Item=(Type=Bone,Name="collar_09"), Space=LocalSpace, bInitial=False, Item.Name=@n1.BoneIn.Name).Transform.Translation) {template="Unit::Execute(in Value,out Result)"}.Result
    A=RigVMVariableNode() {variable=head_vector}.Value
   ) {is=struct RigVMUnitNode, template="Dot,|::Execute(in A,in B,out Result)"}
   n2 = &RigVMFunction_MathDoubleRemap(TargetMaximum=1, TargetMinimum=0, Value=@Dot__.Result, SourceMinimum=@n1.s_min, SourceMaximum=@n1.s_max)
   Multiply = &RigVMFunction_MathDoubleMul(A=@n1.angle, B=@n2.Result)
   > &RigUnit_ModifyTransforms(
    ItemToModify=((Item=(Type=Bone,Name="None"),Transform=(Rotation=(X=0,Y=0,Z=0,W=1),Translation=(X=0,Y=0,Z=0),Scale3D=(X=1,Y=1,Z=1))))
    Weight=1
    WeightMinimum=0
    ItemToModify.0.Transform.Rotation=&RigVMFunction_MathQuaternionFromEuler(Euler=(X=0,Y=0,Z=0), Euler.Z=@Multiply.Result).Result
    ItemToModify.0.Item.Name=@n1.BoneIn.Name
   )
   > RigVMFunctionReturnNode(DotAlpha=@n2.Result, BoneOut=@n1.BoneIn.Name)
  ]}
 RigVMFunctionLibrary {nodes [DotRot_CollarHead = RigVMCollapseNode(BoneIn=(Type=None,Name="None"), ExecuteContext=())]}
 RigVMFunction_Sequence_13_ContainedGraph {nodes [
   n1 = RigVMFunctionEntryNode()
   n2 = &RigVMFunction_Sequence(ExecuteContext=@n1.ExecuteContext)
   n3 = &RigVMFunction_Sequence(ExecuteContext=@n2.B)
   RigVMFunctionReturnNode(B=@n3.A, C=@n3.B, A=@n2.A)
  ]}
 RigVMFunction_Sequence_2_ContainedGraph {nodes [
   n1 = RigVMFunctionEntryNode()
   n2 = &RigVMFunction_Sequence(ExecuteContext=@n1.ExecuteContext)
   n3 = &RigVMFunction_Sequence(ExecuteContext=@n2.B)
   RigVMFunctionReturnNode(A=@n2.A, B=@n3.A, C=@n3.B)
  ]}
 RigVMFunction_Sequence_3_ContainedGraph {nodes [
   n1 = RigVMFunctionEntryNode()
   n2 = &RigVMFunction_Sequence(ExecuteContext=@n1.ExecuteContext)
   n3 = &RigVMFunction_Sequence(ExecuteContext=@n2.B)
   RigVMFunctionReturnNode(B=@n3.A, C=@n3.B, A=@n2.A)
  ]}
 RigVMFunction_Sequence_3_ContainedGraph {nodes [
   n1 = RigVMFunctionEntryNode()
   RigVMFunction_Sequence = &RigVMFunction_Sequence(ExecuteContext=@n1.ExecuteContext)
   n2 = &RigVMFunction_Sequence(ExecuteContext=@RigVMFunction_Sequence.B)
   RigVMFunctionReturnNode(A=@RigVMFunction_Sequence.A, B=@n2.A, C=@n2.B)
  ]}
 RigVMFunction_Sequence_4_ContainedGraph {nodes [
   n1 = RigVMFunctionEntryNode()
   n2 = &RigVMFunction_Sequence(ExecuteContext=@n1.ExecuteContext)
   n3 = &RigVMFunction_Sequence(ExecuteContext=@n2.B)
   RigVMFunctionReturnNode(B=@n3.A, C=@n3.B, A=@n2.A)
  ]}
 RigVMFunction_Sequence_5_ContainedGraph {nodes [
   n1 = RigVMFunctionEntryNode()
   n2 = &RigVMFunction_Sequence(ExecuteContext=@n1.ExecuteContext)
   n3 = &RigVMFunction_Sequence(ExecuteContext=@n2.B)
   n4 = &RigVMFunction_Sequence(ExecuteContext=@n3.B)
   RigVMFunctionReturnNode(B=@n3.A, C=@n4.A, D=@n4.B, A=@n2.A)
  ]}
 RigVMFunction_Sequence_9_ContainedGraph {nodes [
   n1 = RigVMFunctionEntryNode()
   n2 = &RigVMFunction_Sequence(ExecuteContext=@n1.ExecuteContext)
   n3 = &RigVMFunction_Sequence(ExecuteContext=@n2.B)
   RigVMFunctionReturnNode(B=@n3.A, C=@n3.B, A=@n2.A)
  ]}
 RigVMFunction_Sequence_9_ContainedGraph {nodes [
   n1 = RigVMFunctionEntryNode()
   n2 = &RigVMFunction_Sequence(ExecuteContext=@n1.ExecuteContext)
   n3 = &RigVMFunction_Sequence(ExecuteContext=@n2.B)
   n4 = &RigVMFunction_Sequence(ExecuteContext=@n3.B)
   RigVMFunctionReturnNode(B=@n3.A, C=@n4.A, D=@n4.B, A=@n2.A)
  ]}
 RigVMModel {nodes [
   &RigUnit_ParentConstraint(
    AdvancedSettings=(InterpolationType=Average,RotationOrderForFilter=XZY)
    Child=(Type=Bone,Name="dyn_jacket_r_side_base")
    ExecutePin=()
    Parents=((Item=(Type=Bone,Name="dyn_jacket_r_back_base"),Weight=0.5),(Item=(Type=Bone,Name="dyn_jacket_r_fr_base"),Weight=0.5))
    Weight=1
   )
   > &RigUnit_ModifyTransforms(
    ItemToModify=((Item=(Type=Bone,Name="dyn_jacket_r_side_base"),Transform=(Rotation=(X=0,Y=0,Z=-0.1167,W=0.9932),Translation=(X=-0.7008,Y=-2.9401,Z=0),Scale3D=(X=1,Y=1,Z=1))))
    Weight=1
    WeightMinimum=0
   )
   n1 = RigVMRerouteNode(Value=())
   n2 = &RigUnit_SphericalPoseReader(
    ActiveRegionScaleFactors=(PositiveWidth=1,NegativeWidth=3.5,PositiveHeight=3,NegativeHeight=3)
    ActiveRegionSize=0.05
    Debug=(bDrawDebug=true,DebugScale=25,DebugSegments=20,DebugThickness=0.25)
    DriverAxis=(X=1,Y=0,Z=0)
    DriverItem=(Type=Bone,Name="thigh_r")
    FalloffRegionScaleFactors=(PositiveWidth=1,NegativeWidth=1,PositiveHeight=1,NegativeHeight=1)
    FalloffSize=0.2
    OptionalParentItem=(Type=Bone)
    RotationOffset=(X=-47,Y=-14,Z=4)
    ExecutePin=@n1.Value
   )
   > &RigUnit_ParentConstraint(
    AdvancedSettings=(RotationOrderForFilter=XZY)
    Child=(Type=Bone,Name="dyn_jacket_r_fr_base")
    Parents=((Item=(Type=Bone,Name="thigh_r"),Weight=1))
    bMaintainOffset=true
    Weight=@n2.OutputParam
   )
   > &RigUnit_ModifyTransforms(
    ItemToModify=((Item=(Type=Bone,Name="dyn_jacket_r_fr_base"),Transform=(Rotation=(X=-0.1979,Y=-0.6973,Z=0.167,W=0.6684),Translation=(X=-1.3777,Y=-13.2794,Z=1.8488),Scale3D=(X=1,Y=1,Z=1))))
    WeightMinimum=0
    Weight=@n2.OutputParam
   )
   > &RigUnit_ParentConstraint(
    AdvancedSettings=(InterpolationType=Average,RotationOrderForFilter=XZY)
    Child=(Type=Bone,Name="dyn_jacket_r_side_base")
    Parents=((Item=(Type=Bone,Name="dyn_jacket_r_fr_base"),Weight=0.5),(Item=(Type=Bone,Name="dyn_jacket_r_back_base"),Weight=0.5))
    Weight=@n2.OutputParam
   )
   > &RigUnit_ModifyTransforms(
    ItemToModify=((Item=(Type=Bone,Name="dyn_jacket_r_side_base"),Transform=(Rotation=(X=0.0093,Y=0.0607,Z=0.0157,W=-0.998),Translation=(X=0.043,Y=-1.9001,Z=-2.9063),Scale3D=(X=1,Y=1,Z=1))))
    WeightMinimum=0
    Weight=@n2.OutputParam
   )
   > &RigUnit_ModifyTransforms(
    ItemToModify=((Item=(Type=Bone,Name="dyn_belt_fr_base_r"),Transform=(Rotation=(X=-0.0083,Y=-0.1906,Z=-0.0428,W=0.9807),Translation=(X=0,Y=0,Z=0),Scale3D=(X=1,Y=1,Z=1))))
    WeightMinimum=0
    Weight=@n2.OutputParam
   )
   RigUnit_BeginExecution = &RigUnit_BeginExecution()
   RigVMFunction_Sequence = RigVMAggregateNode(ExecuteContext=@RigUnit_BeginExecution.ExecutePin)
   FN_DeformRig_FNC = RigVMFunctionReferenceNode(
    DeformRig_ Curve="(deformRig_l_upperarm_up_alpha_72_#1=False,deformRig_r_upperarm_up_alpha_73_#2=False,deformRig_l_upperarm_fr_alpha_74_#3=False,deformRig_r_upperarm_fr_alpha_75_CB51B8 ...(truncated)"
    DeformRig_Bool="(upperarm_1_#4=True,elbow_3_#5=True,wrist_7_#6=True,groin_8_#7=True,glute_17_#8=True,knee_9_40C408324CA261 ...(truncated)"
    DeformRig_Debug=(upperarm_debug_18_#4=False,wrist_debug_20_#6=False,lat_debug_26_#9=False,groinGlute_debug_27_#7=False)
    DeformRig_Value="(upperarm_up_18_#10=(Rotation=(X=0,Y=0,Z=0,W=1),Translation=(X=0,Y=0,Z=-5),Scale3D=(X=0.7,Y=1,Z=1)),upperarm_fwd_fr_19_#11=(Rotation=(X=0,Y=0,Z=0,W=1),Translation=(X=0,Y=10,Z=0),Scale3D=(X=0.5,Y=1,Z ...(truncated)"
    ExecuteContext=@RigVMFunction_Sequence.A
   )
   RigVMFunction_Sequence_6 = RigVMAggregateNode(ExecuteContext=@RigVMFunction_Sequence.B)
   n3 = &RigVMFunction_Sequence(ExecuteContext=@RigVMFunction_Sequence_6.A)
   &RigUnit_ParentConstraint(
    AdvancedSettings=(RotationOrderForFilter=XZY)
    Child=(Type=Bone,Name="dyn_jacket_l_fr_base")
    Parents="((Item=(Type=Bone,Name=\"dyn_jacket_02_l_fr\"),Weight=0.0044),(Item=(Type=Bone,Name=\"dyn_jacket_01_l_side\"),Weight=0.0079),(Item=(Type=Bone,Name=\"thigh_l\"),Weight=0.0279),(Item=(Type=Bone,Name=\"spine_02\"),Weight=0.0354),(Item=(Type=Bone,Name=\"dyn_jacket_01_l ...(truncated)"
    Weight=1
    ExecutePin=@n3.A
   )
   > n4 = &RigUnit_SphericalPoseReader(
    ActiveRegionScaleFactors=(PositiveWidth=3.5,NegativeWidth=0.5,PositiveHeight=3,NegativeHeight=3)
    ActiveRegionSize=0.1
    Debug=(bDrawDebug=true,DebugScale=25,DebugSegments=20,DebugThickness=0.25)
    DriverAxis=(X=-1,Y=0,Z=0)
    DriverItem=(Type=Bone,Name="thigh_l")
    FalloffRegionScaleFactors=(PositiveWidth=1,NegativeWidth=1,PositiveHeight=1,NegativeHeight=1)
    FalloffSize=0.45
    OptionalParentItem=(Type=Bone)
    RotationOffset=(X=90,Y=0,Z=0)
   )
   > RigVMVariableNode(Value=@n4.OutputParam) {variable=SPR_up_thigh_l}
   n5 = &RigVMFunction_Sequence(ExecuteContext=@n3.B)
   n6 = RigVMVariableNode() {variable=crouch}
   RigVMFunction_MathBoolNot = &RigVMFunction_MathBoolNot(Value=@n6.Value)
   RigVMFunction_MathBoolToFloat = &RigVMFunction_MathBoolToFloat(Value=@RigVMFunction_MathBoolNot.Result)
   Multiply = &RigVMFunction_MathDoubleMul(B=@RigVMFunction_MathBoolToFloat.Result, A=RigVMVariableNode() {variable=SPR_up_thigh_l}.Value)
   &RigUnit_ModifyTransforms(
    ItemToModify=((Item=(Type=Bone,Name="dyn_jacket_l_fr_base"),Transform=(Rotation=(X=0,Y=0,Z=0,W=1),Translation=(X=-10,Y=-5,Z=-5),Scale3D=(X=1,Y=1,Z=1))))
    WeightMinimum=0
    Weight=@Multiply.Result
    ExecutePin=@n5.A
   )
   &RigUnit_ModifyTransforms(
    ItemToModify=((Item=(Type=Bone,Name="dyn_jacket_l_fr_base"),Transform=(Rotation=(X=-0.4202,Y=-0.1139,Z=0.0449,W=-0.8991),Translation=(X=-10.4274,Y=-7.7147,Z=-4.8812),Scale3D=(X=1,Y=1,Z=1))))
    WeightMinimum=0
    Weight=&RigVMFunction_MathDoubleMul(B=&RigVMFunction_MathBoolToFloat(Value=@n6.Value).Result, A=RigVMVariableNode() {variable=SPR_up_thigh_l}.Value).Result
    ExecutePin=@n5.B
   )
   RigVMFunction_Sequence_5 = RigVMAggregateNode(ExecuteContext=@RigVMFunction_Sequence_6.B)
   n7 = &RigVMFunction_Sequence(ExecuteContext=@RigVMFunction_Sequence_5.B)
   n8 = &RigUnit_SphericalPoseReader(
    ActiveRegionScaleFactors=(PositiveWidth=0.5,NegativeWidth=3.5,PositiveHeight=3,NegativeHeight=3)
    ActiveRegionSize=0.05
    Debug=(bDrawDebug=true,DebugScale=25,DebugSegments=20,DebugThickness=0.25)
    DriverAxis=(X=1,Y=0,Z=0)
    DriverItem=(Type=Bone,Name="thigh_r")
    FalloffRegionScaleFactors=(PositiveWidth=1,NegativeWidth=1,PositiveHeight=1,NegativeHeight=1)
    FalloffSize=0.3
    OptionalParentItem=(Type=Bone)
    RotationOffset=(X=-16,Y=14,Z=0)
    ExecutePin=@n7.A
   )
   > &RigUnit_ParentConstraint(
    AdvancedSettings=(RotationOrderForFilter=XZY)
    Child=(Type=Bone,Name="dyn_jacket_r_fr_base")
    Parents=((Item=(Type=Bone,Name="thigh_r"),Weight=1))
    Weight=1
    bMaintainOffset=true
   )
   > &RigUnit_ModifyTransforms(
    ItemToModify=((Item=(Type=Bone,Name="dyn_jacket_r_fr_base"),Transform=(Rotation=(X=-0.0784,Y=-0.1442,Z=0.1966,W=0.9666),Translation=(X=7.9882,Y=-13.3647,Z=9.4829),Scale3D=(X=1,Y=1,Z=1))))
    WeightMinimum=0
    Weight=@n8.OutputParam
   )
   > &RigUnit_ParentConstraint(
    AdvancedSettings=(InterpolationType=Average,RotationOrderForFilter=XZY)
    Child=(Type=Bone,Name="dyn_jacket_r_side_base")
    Parents=((Item=(Type=Bone,Name="dyn_jacket_r_fr_base"),Weight=0.5),(Item=(Type=Bone,Name="dyn_jacket_r_back_base"),Weight=0.5))
    Weight=1
   )
   > &RigUnit_ModifyTransforms(
    ItemToModify=((Item=(Type=Bone,Name="dyn_jacket_r_side_base"),Transform=(Rotation=(X=-0.1057,Y=0.0109,Z=-0.3045,W=0.9466),Translation=(X=-1.4535,Y=0.0021,Z=-4.0285),Scale3D=(X=1,Y=1,Z=1))))
    WeightMinimum=0
    Weight=@n8.OutputParam
   )
   n9 = &RigUnit_SphericalPoseReader(
    ActiveRegionScaleFactors=(PositiveWidth=1,NegativeWidth=3.5,PositiveHeight=3,NegativeHeight=3)
    ActiveRegionSize=0.05
    Debug=(bDrawDebug=true,DebugScale=25,DebugSegments=20,DebugThickness=0.25)
    DriverAxis=(X=1,Y=0,Z=0)
    DriverItem=(Type=Bone,Name="thigh_r")
    FalloffRegionScaleFactors=(PositiveWidth=1,NegativeWidth=1,PositiveHeight=1,NegativeHeight=1)
    FalloffSize=0.2
    OptionalParentItem=(Type=Bone)
    RotationOffset=(X=-47,Y=-14,Z=4)
    ExecutePin=@n7.B
   )
   > &RigUnit_ParentConstraint(
    AdvancedSettings=(RotationOrderForFilter=XZY)
    Child=(Type=Bone,Name="dyn_jacket_r_fr_base")
    Parents=((Item=(Type=Bone,Name="thigh_r"),Weight=1))
    bMaintainOffset=true
    Weight=@n9.OutputParam
   )
   > &RigUnit_ModifyTransforms(
    ItemToModify=((Item=(Type=Bone,Name="dyn_jacket_r_fr_base"),Transform=(Rotation=(X=-0.1979,Y=-0.6973,Z=0.167,W=0.6684),Translation=(X=-1.3777,Y=-13.2794,Z=1.8488),Scale3D=(X=1,Y=1,Z=1))))
    WeightMinimum=0
    Weight=@n9.OutputParam
   )
   > &RigUnit_ParentConstraint(
    AdvancedSettings=(InterpolationType=Average,RotationOrderForFilter=XZY)
    Child=(Type=Bone,Name="dyn_jacket_r_side_base")
    Parents=((Item=(Type=Bone,Name="dyn_jacket_r_fr_base"),Weight=0.5),(Item=(Type=Bone,Name="dyn_jacket_r_back_base"),Weight=0.5))
    Weight=@n9.OutputParam
   )
   > &RigUnit_ModifyTransforms(
    ItemToModify=((Item=(Type=Bone,Name="dyn_jacket_r_side_base"),Transform=(Rotation=(X=0.0093,Y=0.0607,Z=0.0157,W=-0.998),Translation=(X=0.043,Y=-1.9001,Z=-2.9063),Scale3D=(X=1,Y=1,Z=1))))
    WeightMinimum=0
    Weight=@n9.OutputParam
   )
   > &RigUnit_ModifyTransforms(
    ItemToModify=((Item=(Type=Bone,Name="dyn_belt_fr_base_r"),Transform=(Rotation=(X=-0.0083,Y=-0.1906,Z=-0.0428,W=0.9807),Translation=(X=0,Y=0,Z=0),Scale3D=(X=1,Y=1,Z=1))))
    WeightMinimum=0
    Weight=@n9.OutputParam
   )
   &RigUnit_ParentConstraint(
    AdvancedSettings=(RotationOrderForFilter=XZY)
    Child=(Type=Bone,Name="dyn_belt_06_l")
    Parents=((Item=(Type=Bone,Name="spine_04"),Weight=0.0393),(Item=(Type=Bone,Name="spine_03"),Weight=0.0435),(Item=(Type=Bone,Name="spine_01"),Weight=0.1151),(Item=(Type=Bone,Name="spine_02"),Weight=0.3346),(Item=(Type=Bone,Name="pelvis"),Weight=0.4675))
    Weight=1
    ExecutePin=@RigVMFunction_Sequence_5.C
   )
   RigVMFunction_Sequence_7 = RigVMAggregateNode(ExecuteContext=@RigVMFunction_Sequence_6.C)
   &RigUnit_ModifyTransforms(
    ItemToModify=((Item=(Type=Bone,Name="dyn_jacket_r_fr_base"),Transform=(Rotation=(X=0,Y=0,Z=0,W=1),Translation=(X=0,Y=3,Z=-5),Scale3D=(X=1,Y=1,Z=1))))
    WeightMinimum=0
    ExecutePin=@RigVMFunction_Sequence_7.B
    Weight=RigVMVariableNode() {variable=SPR_UP_thigh_r}.Value
   )
   n10 = &RigUnit_SphericalPoseReader(
    ActiveRegionScaleFactors=(PositiveWidth=3.5,NegativeWidth=0.5,PositiveHeight=3,NegativeHeight=3)
    ActiveRegionSize=0.1
    Debug=(bDrawDebug=true,DebugScale=25,DebugSegments=20,DebugThickness=0.25)
    DriverAxis=(X=1,Y=0,Z=0)
    DriverItem=(Type=Bone,Name="thigh_r")
    FalloffRegionScaleFactors=(PositiveWidth=1,NegativeWidth=1,PositiveHeight=1,NegativeHeight=1)
    FalloffSize=0.45
    FlipHeightScaling=false
    FlipWidthScaling=true
    OptionalParentItem=(Type=Bone)
    RotationOffset=(X=-90,Y=0,Z=0)
    ExecutePin=@RigVMFunction_Sequence_7.A
   )
   > RigVMVariableNode(Value=@n10.OutputParam) {variable=SPR_UP_thigh_r}
   RigVMFunction_Sequence_3 = RigVMAggregateNode(ExecuteContext=@RigVMFunction_Sequence.C)
   head_vector_nodes = RigVMCollapseNode()
   RigVMVariableNode(Value=@head_vector_nodes.Result, ExecuteContext=@RigVMFunction_Sequence_3.B) {variable=head_vector}
   > RigVMVariableNode(Value=-90) {variable=rotate_collar}
   > RigVMFunction_Sequence_1 = RigVMAggregateNode()
   DotRot_CollarHead_1 = RigVMFunctionReferenceNode(
    BoneIn=(Type=Bone,Name="extra_collar_01_l")
    s_max=1
    s_min=0.9
    ExecuteContext=@RigVMFunction_Sequence_1.A
    angle=RigVMVariableNode() {variable=rotate_collar}.Value
   )
   > DotRot_CollarHead_2 = RigVMFunctionReferenceNode(BoneIn=(Type=Bone,Name="extra_collar_02_l"), s_max=1, s_min=0.9, angle=RigVMVariableNode() {variable=rotate_collar}.Value)
   > DotRot_CollarHead_3 = RigVMFunctionReferenceNode(BoneIn=(Type=Bone,Name="extra_collar_03_l"), s_max=1, s_min=0.9, angle=RigVMVariableNode() {variable=rotate_collar}.Value)
   > DotRot_CollarHead_5 = RigVMFunctionReferenceNode(BoneIn=(Type=Bone,Name="extra_collar_04_l"), s_max=1, s_min=0.9, angle=RigVMVariableNode() {variable=rotate_collar}.Value)
   > DotRot_CollarHead_4 = RigVMFunctionReferenceNode(BoneIn=(Type=Bone,Name="extra_collar_05_l"), s_max=1, s_min=0.9, angle=RigVMVariableNode() {variable=rotate_collar}.Value)
   > DotRot_CollarHead_6 = RigVMFunctionReferenceNode(BoneIn=(Type=Bone,Name="extra_collar_06_l"), s_max=1, s_min=0.9, angle=RigVMVariableNode() {variable=rotate_collar}.Value)
   DotRot_CollarHead = RigVMFunctionReferenceNode(
    BoneIn=(Type=Bone,Name="extra_collar_m_l")
    s_max=1
    s_min=0.9
    ExecuteContext=@RigVMFunction_Sequence_1.B
    angle=RigVMVariableNode() {variable=rotate_collar}.Value
   )
   DotRot_CollarHead_7 = RigVMFunctionReferenceNode(
    BoneIn=(Type=Bone,Name="extra_collar_01_r")
    s_max=1
    s_min=0.9
    ExecuteContext=@RigVMFunction_Sequence_1.C
    angle=RigVMVariableNode() {variable=rotate_collar}.Value
   )
   > DotRot_CollarHead_8 = RigVMFunctionReferenceNode(BoneIn=(Type=Bone,Name="extra_collar_02_r"), s_max=1, s_min=0.9, angle=RigVMVariableNode() {variable=rotate_collar}.Value)
   > DotRot_CollarHead_9 = RigVMFunctionReferenceNode(BoneIn=(Type=Bone,Name="extra_collar_03_r"), s_max=1, s_min=0.9, angle=RigVMVariableNode() {variable=rotate_collar}.Value)
   > DotRot_CollarHead_11 = RigVMFunctionReferenceNode(BoneIn=(Type=Bone,Name="extra_collar_04_r"), s_max=1, s_min=0.9, angle=RigVMVariableNode() {variable=rotate_collar}.Value)
   > DotRot_CollarHead_10 = RigVMFunctionReferenceNode(BoneIn=(Type=Bone,Name="extra_collar_05_r"), s_max=1, s_min=0.9, angle=RigVMVariableNode() {variable=rotate_collar}.Value)
   > DotRot_CollarHead_12 = RigVMFunctionReferenceNode(BoneIn=(Type=Bone,Name="extra_collar_06_r"), s_max=1, s_min=0.9, angle=RigVMVariableNode() {variable=rotate_collar}.Value)
   RigVMFunction_Sequence_4 = RigVMAggregateNode(ExecuteContext=@RigVMFunction_Sequence_3.C)
   n11 = &RigUnit_SphericalPoseReader(
    ActiveRegionScaleFactors=(PositiveWidth=1,NegativeWidth=0.1,PositiveHeight=0.5,NegativeHeight=1)
    ActiveRegionSize=0.3
    Debug=(bDrawDebug=true,bDraw2D=false,bDrawLocalAxes=false,DebugScale=30,DebugSegments=30,DebugThickness=0.2)
    DriverAxis=(X=-1,Y=0,Z=0)
    DriverItem=(Type=Bone,Name="clavicle_r")
    FalloffRegionScaleFactors=(PositiveWidth=1,NegativeWidth=0.5,PositiveHeight=1,NegativeHeight=1)
    FalloffSize=0.5
    OptionalParentItem=(Type=Bone,Name="None")
    RotationOffset=(X=0,Y=220,Z=180)
    ExecutePin=@RigVMFunction_Sequence_4.C
   )
   n12 = &RigVMFunction_AlphaInterp(
    InterpSpeedDecreasing=10
    InterpSpeedIncreasing=10
    Scale=1
    bClampResult=False
    Value=&RigVMFunction_MathDoubleMul(A=@n11.OutputParam, B=RigVMVariableNode() {variable=targetAlpha}.Value).Result
   )
   > &RigUnit_ModifyTransforms(
    ItemToModify="((Item=(Type=Bone,Name=\"extra_collar_01_r\"),Transform=(Rotation=(X=0,Y=0,Z=0,W=1),Translation=(X=0,Y=0,Z=0),Scale3D=(X=1,Y=1,Z=1))),(Item=(Type=Bone,Name=\"extra_collar_02_r\"),Transform=(Rotation=(X=0,Y=0,Z=0.342,W=0.9397),Translation=(X=0,Y=0,Z=0),Scale3D= ...(truncated)"
    WeightMinimum=0
    Weight=@n12.Result
   )
   n13 = &RigUnit_SphericalPoseReader(
    ActiveRegionScaleFactors=(PositiveWidth=1,NegativeWidth=0.1,PositiveHeight=1,NegativeHeight=0.5)
    ActiveRegionSize=0.3
    Debug=(bDrawDebug=true,bDraw2D=false,bDrawLocalAxes=false,DebugScale=30,DebugSegments=30,DebugThickness=0.2)
    DriverAxis=(X=1,Y=0,Z=0)
    DriverItem=(Type=Bone,Name="clavicle_l")
    FalloffRegionScaleFactors=(PositiveWidth=1,NegativeWidth=0.5,PositiveHeight=1,NegativeHeight=0.5)
    FalloffSize=0.5
    OptionalParentItem=(Type=Bone,Name="None")
    RotationOffset=(X=0,Y=40,Z=180)
    ExecutePin=@RigVMFunction_Sequence_4.B
   )
   AlphaInterp = &RigVMFunction_AlphaInterp(
    InterpSpeedDecreasing=10
    InterpSpeedIncreasing=10
    Scale=1
    bClampResult=False
    Value=&RigVMFunction_MathDoubleMul(A=@n13.OutputParam, B=RigVMVariableNode() {variable=targetAlpha}.Value).Result
   )
   > &RigUnit_ModifyTransforms(
    ItemToModify="((Item=(Type=Bone,Name=\"extra_collar_06_l\"),Transform=(Rotation=(X=0,Y=0,Z=0,W=1),Translation=(X=0,Y=-0.5,Z=0),Scale3D=(X=1,Y=1,Z=1))),(Item=(Type=Bone,Name=\"extra_collar_05_l\"),Transform=(Rotation=(X=0,Y=0,Z=0,W=1),Translation=(X=0,Y=-1,Z=0),Scale3D=(X=1, ...(truncated)"
    WeightMinimum=0
    Weight=@AlphaInterp.Result
   )
   n14 = &RigUnit_SphericalPoseReader(
    ActiveRegionScaleFactors=(PositiveWidth=0.3,NegativeWidth=0.3,PositiveHeight=0.3,NegativeHeight=0.3)
    ActiveRegionSize=0.3
    Debug=(bDrawDebug=true,bDraw2D=false,bDrawLocalAxes=false,DebugScale=30,DebugSegments=30,DebugThickness=0.2)
    DriverAxis=(X=-1,Y=0,Z=0)
    DriverItem=(Type=Bone,Name="upperarm_r")
    FalloffRegionScaleFactors=(PositiveWidth=0.1,NegativeWidth=1,PositiveHeight=1,NegativeHeight=0.5)
    FalloffSize=0.5
    OptionalParentItem=(Type=Bone,Name="None")
    RotationOffset=(X=-40,Y=220,Z=0)
    ExecutePin=@RigVMFunction_Sequence_4.D
   )
   RigVMFunction_ControlFlowBranch = &RigVMFunction_ControlFlowBranch(ExecuteContext=@n14.ExecutePin, Condition=RigVMVariableNode() {variable=crouch}.Value)
   n15 = &RigVMFunction_AlphaInterp(
    InterpSpeedDecreasing=10
    InterpSpeedIncreasing=10
    Scale=1
    bClampResult=False
    Value=&RigVMFunction_MathDoubleMul(A=@n14.OutputParam, B=RigVMVariableNode() {variable=targetAlpha}.Value).Result
   )
   &RigUnit_ModifyTransforms(
    ItemToModify="((Item=(Type=Bone,Name=\"extra_collar_01_r\"),Transform=(Rotation=(X=0,Y=0,Z=0,W=1),Translation=(X=0,Y=0,Z=0),Scale3D=(X=1,Y=1,Z=1))),(Item=(Type=Bone,Name=\"extra_collar_02_r\"),Transform=(Rotation=(X=0,Y=0,Z=-0.1736,W=0.9848),Translation=(X=0,Y=0,Z=0),Scale3 ...(truncated)"
    WeightMinimum=0
    Weight=@n15.Result
    ExecutePin=@RigVMFunction_ControlFlowBranch.False
   )
   &RigUnit_ModifyTransforms(
    ItemToModify="((Item=(Type=Bone,Name=\"extra_collar_01_r\"),Transform=(Rotation=(X=0,Y=0,Z=0,W=1),Translation=(X=0,Y=0,Z=0),Scale3D=(X=1,Y=1,Z=1))),(Item=(Type=Bone,Name=\"extra_collar_02_r\"),Transform=(Rotation=(X=0,Y=0,Z=-0.2588,W=0.9659),Translation=(X=0,Y=0,Z=0),Scale3 ...(truncated)"
    WeightMinimum=0
    Weight=@n15.Result
    ExecutePin=@RigVMFunction_ControlFlowBranch.True
   )
   RigVMVariableNode(
    Value=&RigVMFunction_MathBoolToFloat(Value=RigVMVariableNode() {variable=target}.Value).Result
    ExecuteContext=@RigVMFunction_Sequence_4.A
   ) {variable=targetAlpha}
   RigVMVariableNode(Value=0.4, ExecuteContext=@RigVMFunction_Sequence_3.A) {variable=contrs}
   > RigVMFunction_Sequence_2 = RigVMAggregateNode()
   n16 = RigVMVariableNode() {variable=contrs}
   &RigUnit_ParentConstraint(
    AdvancedSettings=(RotationOrderForFilter=XZY)
    Child=(Type=Bone,Name="extra_collar_01_l")
    Filter=(TranslationFilter=(bX=True,bY=True,bZ=True),RotationFilter=(bX=False,bY=False,bZ=False),ScaleFilter=(bX=True,bY=True,bZ=True))
    Parents=((Item=(Type=Bone,Name="spine_03"),Weight=0.0434),(Item=(Type=Bone,Name="neck_02"),Weight=0.0541),(Item=(Type=Bone,Name="neck_01"),Weight=0.09),(Item=(Type=Bone,Name="clavicle_l"),Weight=0.1092),(Item=(Type=Bone,Name="spine_05"),Weight=0.7034))
    ExecutePin=@RigVMFunction_Sequence_2.A
    Weight=@n16.Value
   )
   > &RigUnit_ParentConstraint(
    AdvancedSettings=(RotationOrderForFilter=XZY)
    Child=(Type=Bone,Name="extra_collar_02_l")
    Filter=(TranslationFilter=(bX=True,bY=True,bZ=True),RotationFilter=(bX=False,bY=False,bZ=False),ScaleFilter=(bX=True,bY=True,bZ=True))
    Parents=((Item=(Type=Bone,Name="spine_03"),Weight=0.0184),(Item=(Type=Bone,Name="neck_01"),Weight=0.0218),(Item=(Type=Bone,Name="clavicle_l"),Weight=0.4081),(Item=(Type=Bone,Name="spine_05"),Weight=0.5516))
    Weight=@n16.Value
   )
   > &RigUnit_ParentConstraint(
    AdvancedSettings=(RotationOrderForFilter=XZY)
    Child=(Type=Bone,Name="extra_collar_03_l")
    Filter=(TranslationFilter=(bX=True,bY=True,bZ=True),RotationFilter=(bX=False,bY=False,bZ=False),ScaleFilter=(bX=True,bY=True,bZ=True))
    Parents=((Item=(Type=Bone,Name="spine_04"),Weight=0.0029),(Item=(Type=Bone,Name="spine_05"),Weight=0.3381),(Item=(Type=Bone,Name="clavicle_l"),Weight=0.659))
    Weight=@n16.Value
   )
   > &RigUnit_ParentConstraint(
    AdvancedSettings=(RotationOrderForFilter=XZY)
    Child=(Type=Bone,Name="extra_collar_04_l")
    Filter=(TranslationFilter=(bX=True,bY=True,bZ=True),RotationFilter=(bX=False,bY=False,bZ=False),ScaleFilter=(bX=True,bY=True,bZ=True))
    Parents="((Item=(Type=Bone,Name=\"spine_03\"),Weight=0.0029),(Item=(Type=Bone,Name=\"neck_02\"),Weight=0.0031),(Item=(Type=Bone,Name=\"neck_01\"),Weight=0.008),(Item=(Type=Bone,Name=\"spine_04\"),Weight=0.012),(Item=(Type=Bone,Name=\"spine_05\"),Weight=0.3943),(Item=(Type=Bo ...(truncated)"
    Weight=&RigVMFunction_MathDoubleMul(B=1.5, A=@n16.Value).Result
   )
   > &RigUnit_ParentConstraint(
    AdvancedSettings=(RotationOrderForFilter=XZY)
    Child=(Type=Bone,Name="extra_collar_05_l")
    Filter=(TranslationFilter=(bX=True,bY=True,bZ=True),RotationFilter=(bX=False,bY=False,bZ=False),ScaleFilter=(bX=True,bY=True,bZ=True))
    Parents="((Item=(Type=Bone,Name=\"deform_pec_l\"),Weight=0.0012),(Item=(Type=Bone,Name=\"upperarm_twist_01_l\"),Weight=0.0012),(Item=(Type=Bone,Name=\"neck_01\"),Weight=0.003),(Item=(Type=Bone,Name=\"spine_04\"),Weight=0.053),(Item=(Type=Bone,Name=\"spine_05\"),Weight=0.4614 ...(truncated)"
    Weight=&RigVMFunction_MathDoubleMul(B=1.75, A=@n16.Value).Result
   )
   > &RigUnit_ParentConstraint(
    AdvancedSettings=(RotationOrderForFilter=XZY)
    Child=(Type=Bone,Name="extra_collar_06_l")
    Filter=(TranslationFilter=(bX=True,bY=True,bZ=True),RotationFilter=(bX=False,bY=False,bZ=False),ScaleFilter=(bX=True,bY=True,bZ=True))
    Parents="((Item=(Type=Bone,Name=\"neck_01\"),Weight=0.0021),(Item=(Type=Bone,Name=\"deform_upperarm_fr_l\"),Weight=0.0096),(Item=(Type=Bone,Name=\"upperarm_twist_01_l\"),Weight=0.0118),(Item=(Type=Bone,Name=\"deform_pec_l\"),Weight=0.0135),(Item=(Type=Bone,Name=\"spine_04\") ...(truncated)"
    Weight=&RigVMFunction_MathDoubleMul(B=2, A=@n16.Value).Result
   )
   &RigUnit_ParentConstraint(
    AdvancedSettings=(RotationOrderForFilter=XZY)
    Child=(Type=Bone,Name="extra_collar_m_l")
    Filter=(TranslationFilter=(bX=True,bY=True,bZ=True),RotationFilter=(bX=False,bY=False,bZ=False),ScaleFilter=(bX=True,bY=True,bZ=True))
    Parents="((Item=(Type=Bone,Name=\"clavicle_r\"),Weight=0.0064),(Item=(Type=Bone,Name=\"clavicle_l\"),Weight=0.0073),(Item=(Type=Bone,Name=\"head\"),Weight=0.0178),(Item=(Type=Bone,Name=\"spine_03\"),Weight=0.0495),(Item=(Type=Bone,Name=\"neck_02\"),Weight=0.1225),(Item=(Type ...(truncated)"
    ExecutePin=@RigVMFunction_Sequence_2.B
    Weight=@n16.Value
   )
   &RigUnit_ParentConstraint(
    AdvancedSettings=(RotationOrderForFilter=XZY)
    Child=(Type=Bone,Name="extra_collar_01_r")
    Filter=(TranslationFilter=(bX=True,bY=True,bZ=True),RotationFilter=(bX=False,bY=False,bZ=False),ScaleFilter=(bX=True,bY=True,bZ=True))
    Parents=((Item=(Type=Bone,Name="spine_03"),Weight=0.0433),(Item=(Type=Bone,Name="neck_02"),Weight=0.0541),(Item=(Type=Bone,Name="neck_01"),Weight=0.09),(Item=(Type=Bone,Name="clavicle_r"),Weight=0.1093),(Item=(Type=Bone,Name="spine_05"),Weight=0.7033))
    ExecutePin=@RigVMFunction_Sequence_2.C
    Weight=@n16.Value
   )
   > &RigUnit_ParentConstraint(
    AdvancedSettings=(RotationOrderForFilter=XZY)
    Child=(Type=Bone,Name="extra_collar_02_r")
    Filter=(TranslationFilter=(bX=True,bY=True,bZ=True),RotationFilter=(bX=False,bY=False,bZ=False),ScaleFilter=(bX=True,bY=True,bZ=True))
    Parents=((Item=(Type=Bone,Name="spine_03"),Weight=0.0184),(Item=(Type=Bone,Name="neck_01"),Weight=0.0218),(Item=(Type=Bone,Name="clavicle_r"),Weight=0.4081),(Item=(Type=Bone,Name="spine_05"),Weight=0.5516))
    Weight=@n16.Value
   )
   > &RigUnit_ParentConstraint(
    AdvancedSettings=(RotationOrderForFilter=XZY)
    Child=(Type=Bone,Name="extra_collar_03_r")
    Filter=(TranslationFilter=(bX=True,bY=True,bZ=True),RotationFilter=(bX=False,bY=False,bZ=False),ScaleFilter=(bX=True,bY=True,bZ=True))
    Parents=((Item=(Type=Bone,Name="spine_05"),Weight=0.3462),(Item=(Type=Bone,Name="clavicle_r"),Weight=0.6538))
    Weight=@n16.Value
   )
   > &RigUnit_ParentConstraint(
    AdvancedSettings=(RotationOrderForFilter=XZY)
    Child=(Type=Bone,Name="extra_collar_04_r")
    Filter=(TranslationFilter=(bX=True,bY=True,bZ=True),RotationFilter=(bX=False,bY=False,bZ=False),ScaleFilter=(bX=True,bY=True,bZ=True))
    Parents="((Item=(Type=Bone,Name=\"spine_03\"),Weight=0.0034),(Item=(Type=Bone,Name=\"neck_02\"),Weight=0.0035),(Item=(Type=Bone,Name=\"neck_01\"),Weight=0.0088),(Item=(Type=Bone,Name=\"spine_04\"),Weight=0.0126),(Item=(Type=Bone,Name=\"spine_05\"),Weight=0.4027),(Item=(Type= ...(truncated)"
    Weight=&RigVMFunction_MathDoubleMul(B=1.5, A=@n16.Value).Result
   )
   > &RigUnit_ParentConstraint(
    AdvancedSettings=(RotationOrderForFilter=XZY)
    Child=(Type=Bone,Name="extra_collar_05_r")
    Filter=(TranslationFilter=(bX=True,bY=True,bZ=True),RotationFilter=(bX=False,bY=False,bZ=False),ScaleFilter=(bX=True,bY=True,bZ=True))
    Parents="((Item=(Type=Bone,Name=\"deform_pec_r\"),Weight=0.0012),(Item=(Type=Bone,Name=\"upperarm_twist_01_r\"),Weight=0.0012),(Item=(Type=Bone,Name=\"neck_01\"),Weight=0.003),(Item=(Type=Bone,Name=\"spine_04\"),Weight=0.053),(Item=(Type=Bone,Name=\"spine_05\"),Weight=0.4614 ...(truncated)"
    Weight=&RigVMFunction_MathDoubleMul(B=1.75, A=@n16.Value).Result
   )
   > &RigUnit_ParentConstraint(
    AdvancedSettings=(RotationOrderForFilter=XZY)
    Child=(Type=Bone,Name="extra_collar_06_r")
    Filter=(TranslationFilter=(bX=True,bY=True,bZ=True),RotationFilter=(bX=False,bY=False,bZ=False),ScaleFilter=(bX=True,bY=True,bZ=True))
    Parents="((Item=(Type=Bone,Name=\"neck_01\"),Weight=0.0021),(Item=(Type=Bone,Name=\"deform_upperarm_fr_r\"),Weight=0.0096),(Item=(Type=Bone,Name=\"upperarm_twist_01_r\"),Weight=0.0118),(Item=(Type=Bone,Name=\"deform_pec_r\"),Weight=0.0135),(Item=(Type=Bone,Name=\"spine_04\") ...(truncated)"
    Weight=&RigVMFunction_MathDoubleMul(B=2, A=@n16.Value).Result
   )
   n17 = &RigVMFunction_Sequence(ExecuteContext=())
   n18 = &RigUnit_SphericalPoseReader(
    ActiveRegionScaleFactors=(PositiveWidth=3.5,NegativeWidth=0.5,PositiveHeight=3,NegativeHeight=3)
    ActiveRegionSize=0.03
    Debug=(bDrawDebug=true,DebugScale=25,DebugSegments=20,DebugThickness=0.25)
    DriverAxis=(X=-1,Y=0,Z=0)
    DriverItem=(Type=Bone,Name="thigh_l")
    FalloffRegionScaleFactors=(PositiveWidth=1,NegativeWidth=1,PositiveHeight=1,NegativeHeight=1)
    FalloffSize=0.2
    OptionalParentItem=(Type=Bone)
    RotationOffset=(X=164,Y=12,Z=0)
    ExecutePin=@n17.A
   )
   > &RigUnit_ModifyTransforms(
    ItemToModify=((Item=(Type=Bone,Name="dyn_jacket_l_fr_base"),Transform=(Rotation=(X=0.0913,Y=0.2893,Z=-0.2962,W=0.9057),Translation=(X=-1.7678,Y=-3.8265,Z=-6.964),Scale3D=(X=1,Y=1,Z=1))))
    WeightMinimum=0
    Weight=@n18.OutputParam
   )
   > &RigUnit_ParentConstraint(
    AdvancedSettings=(InterpolationType=Average,RotationOrderForFilter=XZY)
    Child=(Type=Bone,Name="dyn_jacket_l_side_base")
    Parents=((Item=(Type=Bone,Name="dyn_jacket_l_fr_base"),Weight=0.5),(Item=(Type=Bone,Name="dyn_jacket_l_back_base"),Weight=0.5))
    Weight=@n18.OutputParam
   )
   > &RigUnit_ModifyTransforms(
    ItemToModify=((Item=(Type=Bone,Name="dyn_jacket_l_side_base"),Transform=(Rotation=(X=0.0399,Y=-0.1676,Z=0.2293,W=-0.958),Translation=(X=0,Y=0,Z=0),Scale3D=(X=1,Y=1,Z=1))))
    WeightMinimum=0
    Weight=@n18.OutputParam
   )
   n19 = &RigUnit_SphericalPoseReader(
    ActiveRegionScaleFactors=(PositiveWidth=3.5,NegativeWidth=0.5,PositiveHeight=3,NegativeHeight=3)
    ActiveRegionSize=0.03
    Debug=(bDrawDebug=true,DebugScale=25,DebugSegments=20,DebugThickness=0.25)
    DriverAxis=(X=-1,Y=0,Z=0)
    DriverItem=(Type=Bone,Name="thigh_l")
    FalloffRegionScaleFactors=(PositiveWidth=1,NegativeWidth=1,PositiveHeight=1,NegativeHeight=1)
    FalloffSize=0.2
    OptionalParentItem=(Type=Bone)
    RotationOffset=(X=128,Y=8,Z=0)
    ExecutePin=@n17.B
   )
   > &RigUnit_ModifyTransforms(
    ItemToModify=((Item=(Type=Bone,Name="dyn_jacket_l_fr_base"),Transform=(Rotation=(X=0,Y=0,Z=-0.312,W=0.9501),Translation=(X=0,Y=0,Z=0),Scale3D=(X=1,Y=1,Z=1))))
    WeightMinimum=0
    Weight=@n19.OutputParam
   )
   > &RigUnit_ParentConstraint(
    AdvancedSettings=(InterpolationType=Average,RotationOrderForFilter=XZY)
    Child=(Type=Bone,Name="dyn_jacket_l_side_base")
    Parents=((Item=(Type=Bone,Name="dyn_jacket_l_fr_base"),Weight=0.5),(Item=(Type=Bone,Name="dyn_jacket_l_back_base"),Weight=0.5))
    Weight=@n19.OutputParam
   )
   > &RigUnit_ModifyTransforms(
    ItemToModify=((Item=(Type=Bone,Name="dyn_jacket_l_side_base"),Transform=(Rotation=(X=0.0399,Y=-0.1676,Z=0.2293,W=-0.958),Translation=(X=0,Y=0,Z=0),Scale3D=(X=1,Y=1,Z=1))))
    WeightMinimum=0
    Weight=@n19.OutputParam
   )
   RigVMDispatchNode(Value=1) {template="DISPATCH_RigVMDispatch_Constant(io Value)"}
   RigVMDispatchNode(Value=1) {template="DISPATCH_RigVMDispatch_Constant(io Value)"}
   &RigUnit_ModifyTransforms(
    ExecutePin=()
    ItemToModify=((Item=(Type=Bone,Name="dyn_jacket_l_fr_base"),Transform=(Rotation=(X=0,Y=0,Z=0,W=1),Translation=(X=-10,Y=-5,Z=-5),Scale3D=(X=1,Y=1,Z=1))))
    Weight=1
    WeightMinimum=0
   )
   RigVMRerouteNode(Value=@RigVMFunction_Sequence_5.A)
   RigVMRerouteNode(Value=@RigVMFunction_Sequence_7.C)
  ]}
]
hierarchy [
 root {loc=<0 0 0>, rot=<0 0 0 1>}
 pelvis {parent=root, loc=<0 -0.5301 98.0571>, rot=<0.0127 0.707 -0.0127 -0.707>}
 spine_01 {parent=pelvis, loc=<6.1208 -0.9328 0>, rot=<0 0 0.034 -0.9994>}
 spine_02 {parent=spine_01, loc=<6.5997 -0.2095 0>, rot=<0 0 0.0238 -0.9997>}
 spine_03 {parent=spine_02, loc=<6.4914 -0.1237 0>, rot=<0 0 0.0149 -0.9999>}
 spine_04 {parent=spine_03, loc=<6.9598 0.0794 0>, rot=<0 0 -0.0132 -0.9999>}
 spine_05 {parent=spine_04, loc=<6.9987 0 0>, rot=<0 0 -0.0015 -1>}
 neck_01 {parent=spine_05, loc=<11.3138 0.1529 0>, rot=<0 0 -0.2003 -0.9797>}
 neck_02 {parent=neck_01, loc=<4.5627 0 0>, rot=<0 0 0.0504 -0.9987>}
 head {parent=neck_02, loc=<5.3914 0 0>, rot=<0 0 0.115 -0.9934>}
 FACIAL_C_FacialRoot {parent=head, loc=<2.1524 -0.746 -0.0002>, rot=<0.4879 -0.5118 -0.4879 -0.5118>}
 FACIAL_C_UnderChin {parent=FACIAL_C_FacialRoot, loc=<-0.0005 8.5709 3.8336>, rot=<0.4988 0 0 -0.8667>}
 FACIAL_L_UnderChin {parent=FACIAL_C_FacialRoot, loc=<2.7503 7.3986 2.9191>, rot=<0.4015 -0.355 0.0819 -0.8403>}
 FACIAL_R_UnderChin {parent=FACIAL_C_FacialRoot, loc=<-2.7509 7.3979 2.9217>, rot=<0.4015 0.3549 -0.0819 -0.8403>}
 FACIAL_C_Jaw {parent=FACIAL_C_FacialRoot, loc=<-0.0014 2.4518 2.0283>, rot=<0.3409 0 0 -0.9401>}
 FACIAL_C_Jawline {parent=FACIAL_C_Jaw, loc=<0.0007 2.2502 7.0022>, rot=<0.2586 0 0 -0.966>}
 FACIAL_L_Jawline {parent=FACIAL_C_FacialRoot, loc=<3.9004 5.5494 4.2859>, rot=<0.2082 0.1578 -0.0544 -0.9637>}
 FACIAL_R_Jawline {parent=FACIAL_C_FacialRoot, loc=<-3.9018 5.5502 4.2874>, rot=<0.2082 -0.1578 0.0545 -0.9637>}
 head_end {parent=head, loc=<10 0 0>, rot=<0 0 0 1>}
 clavicle_l {parent=spine_05, loc=<5.4831 -0.1053 -3.2946>, rot=<0.0465 -0.7289 -0.009 -0.6829>}
 upperarm_l {parent=clavicle_l, loc=<9.8658 -1.2136 -0.4105>, rot=<0.0164 -0.3245 -0.0269 -0.9454>}
 lowerarm_l {parent=upperarm_l, loc=<26.465 -0.8345 0.259>, rot=<0.0036 0.0161 -0.2298 -0.9731>}
 hand_l {parent=lowerarm_l, loc=<22.7355 -0.3203 -0.6105>, rot=<-0.6784 -0.0223 0.0735 -0.7307>}
 thumb_01_l {parent=hand_l, loc=<1.3118 -0.3191 -2.459>, rot=<0.6546 -0.4042 0.0664 -0.6354>}
 thumb_02_l {parent=thumb_01_l, loc=<5.9336 -0.0448 0.0401>, rot=<-0.0037 0.0213 0.298 -0.9543>}
 thumb_03_l {parent=thumb_02_l, loc=<2.2223 0.3682 -0.0907>, rot=<0.0118 -0.013 0.0066 -0.9998>}
 index_metacarpal_l {parent=hand_l, loc=<1.6535 0.0183 -2.96>, rot=<0.0158 -0.0124 -0.0173 -0.9996>}
 index_01_l {parent=index_metacarpal_l, loc=<8.3008 0.0357 -0.0524>, rot=<-0.0044 0.0485 0.0924 -0.9945>}
 index_02_l {parent=index_01_l, loc=<3.3587 -0.3004 0.0452>, rot=<-0.002 0.0027 0.3225 -0.9466>}
 index_03_l {parent=index_02_l, loc=<2.5578 -0.123 -0.0252>, rot=<-0.002 0.0067 0.2558 -0.9667>}
 middle_metacarpal_l {parent=hand_l, loc=<2.1856 -0.1725 -0.5984>, rot=<0.006 0.0168 -0.0329 -0.9993>}
 middle_01_l {parent=middle_metacarpal_l, loc=<7.3275 -0.2447 0.0129>, rot=<-0.0208 0.0042 0.0858 -0.9961>}
 middle_02_l {parent=middle_01_l, loc=<4.0123 0.1401 0.1175>, rot=<-0.0218 0.0258 0.3512 -0.9357>}
 middle_03_l {parent=middle_02_l, loc=<2.6077 0.0534 0.1373>, rot=<0.0082 0.0188 0.1708 -0.9851>}
 ring_metacarpal_l {parent=hand_l, loc=<2.2074 -0.3625 1.3638>, rot=<-0.0284 0.0722 -0.0458 -0.9959>}
 ring_01_l {parent=ring_metacarpal_l, loc=<7.0295 -0.6197 -0.1446>, rot=<-0.0489 -0.0252 0.1892 -0.9804>}
 ring_02_l {parent=ring_01_l, loc=<3.6174 0.4647 -0.1027>, rot=<0 0 0.2806 -0.9598>}
 ring_03_l {parent=ring_02_l, loc=<2.4152 0.114 0.0199>, rot=<0 0 0.1189 -0.9929>}
 pinky_metacarpal_l {parent=hand_l, loc=<1.7616 -0.9753 3.3814>, rot=<-0.0397 0.0814 -0.069 -0.9935>}
 pinky_01_l {parent=pinky_metacarpal_l, loc=<7.1709 -0.4831 -0.1094>, rot=<-0.0654 -0.0035 0.2242 -0.9723>}
 pinky_02_l {parent=pinky_01_l, loc=<2.8274 0.0893 -0.1043>, rot=<0.0037 0.0011 0.2854 -0.9584>}
 pinky_03_l {parent=pinky_02_l, loc=<2.2046 0.0307 -0.0078>, rot=<-0.0002 0.0098 0.1445 -0.9895>}
 deform_wrist_dn_base_l {parent=hand_l, loc=<0 -1.5326 0>, rot=<0 0 0 1>}
 deform_wrist_dn_l {parent=deform_wrist_dn_base_l, loc=<0 0 0>, rot=<0 0 0 1>}
 deform_wrist_up_base_l {parent=hand_l, loc=<0 2.3009 0>, rot=<0 0 0 1>}
 deform_wrist_up_l {parent=deform_wrist_up_base_l, loc=<0 0 0>, rot=<0 0 0 1>}
 lowerarm_twist_02_l {parent=lowerarm_l, loc=<10.6707 0 0>, rot=<0 0 0 1>}
 lowerarm_twist_01_l {parent=lowerarm_l, loc=<18.316 0 0>, rot=<0 0 0 1>}
 deform_elbow_in_base_l {parent=lowerarm_l, loc=<0.1834 3.6062 -0.2572>, rot=<0 0 0 1>}
 deform_elbow_in_l {parent=deform_elbow_in_base_l, loc=<0 0 0>, rot=<0 0 0 1>}
 deform_elbow_out_base_l {parent=lowerarm_l, loc=<0 -4 0>, rot=<0 0 0 1>}
 deform_elbow_out_l {parent=deform_elbow_out_base_l, loc=<0 0 0>, rot=<0 0 0 1>}
 upperarm_twist_01_l {parent=upperarm_l, loc=<7.2135 -0.1134 0>, rot=<0 0 0 1>}
 deform_upperarm_bk_base_l {parent=upperarm_twist_01_l, loc=<-6.4053 -4.3746 0.314>, rot=<0 0 0 1>}
 deform_upperarm_bk_l {parent=deform_upperarm_bk_base_l, loc=<0 0 0>, rot=<0 0 0 1>}
 deform_upperarm_fr_base_l {parent=upperarm_twist_01_l, loc=<-5.6312 5.092 0.3354>, rot=<0 0 0 1>}
 deform_upperarm_fr_l {parent=deform_upperarm_fr_base_l, loc=<0 0 0>, rot=<0 0 0 1>}
 deform_upperarm_up_base_l {parent=upperarm_twist_01_l, loc=<-8.891 0.4825 3.9517>, rot=<0 0 0 1>}
 deform_upperarm_up_l {parent=deform_upperarm_up_base_l, loc=<0 0 0>, rot=<0 0 0 1>}
 upperarm_twist_02_l {parent=upperarm_l, loc=<19.7998 -0.1134 0>, rot=<0 0 0 1>}
 clavicle_r {parent=spine_05, loc=<5.4831 -0.1053 3.2946>, rot=<0.7289 0.0465 -0.6829 0.009>}
 upperarm_r {parent=clavicle_r, loc=<-9.8658 1.2136 0.4105>, rot=<0.0164 -0.3245 -0.0269 -0.9454>}
 lowerarm_r {parent=upperarm_r, loc=<-26.465 0.8345 -0.259>, rot=<0.0036 0.0161 -0.2298 -0.9731>}
 hand_r {parent=lowerarm_r, loc=<-22.7355 0.3203 0.6106>, rot=<-0.6784 -0.0223 0.0735 -0.7307>}
 thumb_01_r {parent=hand_r, loc=<-1.3118 0.3191 2.459>, rot=<0.6546 -0.4042 0.0664 -0.6354>}
 thumb_02_r {parent=thumb_01_r, loc=<-5.9336 0.0448 -0.0401>, rot=<-0.0037 0.0213 0.298 -0.9543>}
 thumb_03_r {parent=thumb_02_r, loc=<-2.2223 -0.3682 0.0907>, rot=<0.0118 -0.013 0.0066 -0.9998>}
 index_metacarpal_r {parent=hand_r, loc=<-1.6535 -0.0183 2.96>, rot=<0.0158 -0.0124 -0.0173 -0.9996>}
 index_01_r {parent=index_metacarpal_r, loc=<-8.3008 -0.0357 0.0524>, rot=<-0.0044 0.0485 0.0924 -0.9945>}
 index_02_r {parent=index_01_r, loc=<-3.3587 0.3004 -0.0452>, rot=<-0.002 0.0027 0.3225 -0.9466>}
 index_03_r {parent=index_02_r, loc=<-2.5578 0.123 0.0252>, rot=<-0.002 0.0067 0.2558 -0.9667>}
 middle_metacarpal_r {parent=hand_r, loc=<-2.1856 0.1725 0.5984>, rot=<0.006 0.0168 -0.0329 -0.9993>}
 middle_01_r {parent=middle_metacarpal_r, loc=<-7.3275 0.2447 -0.0129>, rot=<-0.0208 0.0042 0.0858 -0.9961>}
 middle_02_r {parent=middle_01_r, loc=<-4.0123 -0.1401 -0.1175>, rot=<-0.0218 0.0258 0.3512 -0.9357>}
 middle_03_r {parent=middle_02_r, loc=<-2.6077 -0.0534 -0.1373>, rot=<0.0082 0.0188 0.1708 -0.9851>}
 ring_metacarpal_r {parent=hand_r, loc=<-2.2074 0.3625 -1.3638>, rot=<-0.0284 0.0722 -0.0458 -0.9959>}
 ring_01_r {parent=ring_metacarpal_r, loc=<-7.0295 0.6197 0.1446>, rot=<-0.0489 -0.0252 0.1892 -0.9804>}
 ring_02_r {parent=ring_01_r, loc=<-3.6174 -0.4647 0.1027>, rot=<0 0 0.2806 -0.9598>}
 ring_03_r {parent=ring_02_r, loc=<-2.4152 -0.114 -0.0199>, rot=<0 0 0.1189 -0.9929>}
 pinky_metacarpal_r {parent=hand_r, loc=<-1.7616 0.9753 -3.3814>, rot=<-0.0397 0.0814 -0.069 -0.9935>}
 pinky_01_r {parent=pinky_metacarpal_r, loc=<-7.1709 0.4831 0.1094>, rot=<-0.0654 -0.0035 0.2242 -0.9723>}
 pinky_02_r {parent=pinky_01_r, loc=<-2.8274 -0.0893 0.1043>, rot=<0.0037 0.0011 0.2854 -0.9584>}
 pinky_03_r {parent=pinky_02_r, loc=<-2.2046 -0.0307 0.0078>, rot=<-0.0002 0.0098 0.1445 -0.9895>}
 deform_wrist_dn_base_r {parent=hand_r, loc=<-0.0001 1.5328 0>, rot=<0 0 0 1>}
 deform_wrist_dn_r {parent=deform_wrist_dn_base_r, loc=<0 0 0>, rot=<0 0 0 1>}
 deform_wrist_up_base_r {parent=hand_r, loc=<-0.0002 -2.3007 -0.0001>, rot=<0 0 0 1>}
 deform_wrist_up_r {parent=deform_wrist_up_base_r, loc=<0 0 0>, rot=<0 0 0 1>}
 lowerarm_twist_02_r {parent=lowerarm_r, loc=<-10.6707 0 0>, rot=<0 0 0 -1>}
 lowerarm_twist_01_r {parent=lowerarm_r, loc=<-18.316 0 0>, rot=<0 0 0 -1>}
 deform_elbow_in_base_r {parent=lowerarm_r, loc=<-0.1834 -3.6062 0.2572>, rot=<0 0 0 -1>}
 deform_elbow_in_r {parent=deform_elbow_in_base_r, loc=<0 0 0>, rot=<0 0 0 1>}
 deform_elbow_out_base_r {parent=lowerarm_r, loc=<0 4 -0.0001>, rot=<0 0 0 -1>}
 deform_elbow_out_r {parent=deform_elbow_out_base_r, loc=<0 0 0>, rot=<0 0 0 1>}
 upperarm_twist_01_r {parent=upperarm_r, loc=<-7.2135 0.1134 0>, rot=<0 0 0 -1>}
 deform_upperarm_bk_base_r {parent=upperarm_twist_01_r, loc=<6.4053 4.3746 -0.314>, rot=<0 0 0 1>}
 deform_upperarm_bk_r {parent=deform_upperarm_bk_base_r, loc=<0 0 0>, rot=<0 0 0 1>}
 deform_upperarm_fr_base_r {parent=upperarm_twist_01_r, loc=<5.6312 -5.092 -0.3354>, rot=<0 0 0 1>}
 deform_upperarm_fr_r {parent=deform_upperarm_fr_base_r, loc=<0 0 0>, rot=<0 0 0 1>}
 deform_upperarm_up_base_r {parent=upperarm_twist_01_r, loc=<8.891 -0.4825 -3.9517>, rot=<0 0 0 1>}
 deform_upperarm_up_r {parent=deform_upperarm_up_base_r, loc=<0 0 0>, rot=<0 0 0 1>}
 upperarm_twist_02_r {parent=upperarm_r, loc=<-19.7998 0.1134 0>, rot=<0 0 0 -1>}
 attach_backpack {parent=spine_05, loc=<-0.123 -13.473 0>, rot=<-0.0093 -0.707 0.0093 -0.707>}
 Backpack_BR {parent=attach_backpack, type=Null, loc=<0 3.0557 0>, rot=<0 0 0 1>}
 attach_cape {parent=spine_05, loc=<9.017 -11.714 0>, rot=<0.0616 -0.7044 -0.0616 -0.7044>}
 cape {parent=attach_cape, type=Null, loc=<0 0 0>, rot=<0 0 0 1>}
 extra_collar_m_l {parent=spine_05, loc=<12.6369 -5.4723 0>, rot=<-0.0062 -0.0059 -0.0389 -0.9992>}
 extra_collar_01_l {parent=spine_05, loc=<12.5533 -4.8643 -3.158>, rot=<-0.2006 -0.0196 -0.004 -0.9795>}
 extra_collar_02_l {parent=spine_05, loc=<12.2933 -3.3258 -5.7955>, rot=<-0.3845 -0.0216 -0.0279 -0.9225>}
 extra_collar_03_l {parent=spine_05, loc=<11.8508 -1.1385 -7.2707>, rot=<-0.5579 -0.0456 -0.0093 -0.8286>}
 extra_collar_04_l {parent=spine_05, loc=<10.7159 1.5492 -7.8982>, rot=<-0.6326 -0.0894 0.0364 -0.7685>}
 extra_collar_05_l {parent=spine_05, loc=<9.3393 3.7109 -7.7779>, rot=<-0.6847 -0.1327 0.0156 -0.7165>}
 extra_collar_06_l {parent=spine_05, loc=<7.6574 5.9588 -7.6054>, rot=<-0.6271 -0.1331 -0.0139 -0.7674>}
 extra_collar_01_r {parent=spine_05, loc=<12.5532 -4.8643 3.158>, rot=<0.2006 0.0196 -0.004 -0.9795>}
 extra_collar_02_r {parent=spine_05, loc=<12.293 -3.3258 5.7955>, rot=<0.3845 0.0216 -0.0279 -0.9225>}
 extra_collar_03_r {parent=spine_05, loc=<11.8503 -1.1385 7.2707>, rot=<0.5579 0.0456 -0.0093 -0.8286>}
 extra_collar_04_r {parent=spine_05, loc=<10.7163 1.5492 7.8982>, rot=<0.6326 0.0894 0.0364 -0.7685>}
 extra_collar_05_r {parent=spine_05, loc=<9.3397 3.711 7.7779>, rot=<0.6847 0.1327 0.0156 -0.7165>}
 extra_collar_06_r {parent=spine_05, loc=<7.6571 5.9588 7.6054>, rot=<0.6271 0.1331 -0.0139 -0.7674>}
 deform_pec_aim_base_l {parent=spine_04, loc=<3.2456 9.3309 -1.2237>, rot=<0.0603 -0.4033 0.3318 -0.8507>}
 deform_pec_aim_l {parent=deform_pec_aim_base_l, loc=<0 0 0>, rot=<0 0 0 1>}
 deform_pec_base_l {parent=deform_pec_aim_l, loc=<9.3665 0 0>, rot=<0 0 0 1>}
 deform_pec_l {parent=deform_pec_base_l, loc=<0 0 0>, rot=<0 0 0 1>}
 deform_pec_aim_base_r {parent=spine_04, loc=<3.2452 9.3309 1.2237>, rot=<-0.4033 -0.0603 0.8507 0.3318>}
 deform_pec_aim_r {parent=deform_pec_aim_base_r, loc=<0 0 0>, rot=<0 0 0 1>}
 deform_pec_base_r {parent=deform_pec_aim_r, loc=<-9.3667 -0.0003 -0.0004>, rot=<0 0 0 1>}
 deform_pec_r {parent=deform_pec_base_r, loc=<0 0 0>, rot=<0 0 0 1>}
 thigh_l {parent=pelvis, loc=<-1.0855 -0.2109 -8.7053>, rot=<0.0009 0.0355 0.0155 -0.9993>}
 calf_l {parent=thigh_l, loc=<-42.5284 -0.0019 -0.7297>, rot=<0.0006 -0.0161 -0.0391 -0.9991>}
 foot_l {parent=calf_l, loc=<-46.7401 -2.5014 -0.592>, rot=<0.0427 -0.0238 0.0353 -0.9982>}
 ball_l {parent=foot_l, loc=<-3.5658 10.8066 -0.1544>, rot=<0.0001 -0.0001 -0.7063 -0.7079>}
 calf_twist_02_l {parent=calf_l, loc=<-19.1546 -1.7943 -0.2638>, rot=<0.0237 0.0095 -0.0035 -0.9997>}
 calf_twist_01_l {parent=calf_l, loc=<-35.8483 -1.9197 -0.5794>, rot=<0.0237 0.0095 -0.0035 -0.9997>}
 deform_knee_in_base_l {parent=calf_l, loc=<0.8031 -5.0231 0.2456>, rot=<0 0 0 -1>}
 deform_knee_in_l {parent=deform_knee_in_base_l, loc=<0 0 0>, rot=<0 0 0 1>}
 deform_knee_out_base_l {parent=calf_l, loc=<0 4 0>, rot=<0 0 0 -1>}
 deform_knee_out_l {parent=deform_knee_out_base_l, loc=<0 0 0>, rot=<0 0 0 1>}
 extra_knee_pad {parent=calf_l, loc=<-2.2879 4.0878 0.0831>, rot=<0.4049 0.5167 0.5691 0.4951>}
 thigh_twist_01_l {parent=thigh_l, loc=<-27.952 0 0>, rot=<0.0013 0 0 -1>}
 deform_groin_base_l {parent=thigh_l, loc=<0.4801 5.6776 -0.2486>, rot=<0 0 0 -1>}
 deform_groin_l {parent=deform_groin_base_l, loc=<0 0 0>, rot=<0 0 0 1>}
 thigh_r {parent=pelvis, loc=<-1.0855 -0.2109 8.7053>, rot=<0.0355 -0.0009 0.9993 0.0155>}
 calf_r {parent=thigh_r, loc=<42.5284 0.0019 0.7297>, rot=<0.0006 -0.0161 -0.0391 -0.9991>}
 foot_r {parent=calf_r, loc=<46.7401 2.5014 0.592>, rot=<0.0427 -0.0238 0.0353 -0.9982>}
 ball_r {parent=foot_r, loc=<3.5658 -10.8066 0.1544>, rot=<0.0001 -0.0001 -0.7063 -0.7079>}
 calf_twist_02_r {parent=calf_r, loc=<19.1546 1.7943 0.2638>, rot=<0.0237 0.0095 -0.0035 -0.9997>}
 calf_twist_01_r {parent=calf_r, loc=<35.8483 1.9197 0.5794>, rot=<0.0237 0.0095 -0.0035 -0.9997>}
 deform_knee_in_base_r {parent=calf_r, loc=<-0.8031 5.0231 -0.2457>, rot=<0 0 0 -1>}
 deform_knee_in_r {parent=deform_knee_in_base_r, loc=<0 0 0>, rot=<0 0 0 1>}
 deform_knee_out_base_r {parent=calf_r, loc=<0 -4 0>, rot=<0 0 0 -1>}
 deform_knee_out_r {parent=deform_knee_out_base_r, loc=<0 0 0>, rot=<0 0 0 1>}
 thigh_twist_01_r {parent=thigh_r, loc=<27.952 0 0>, rot=<0.0013 0 0 -1>}
 deform_groin_base_r {parent=thigh_r, loc=<-0.4801 -5.6776 0.2486>, rot=<0 0 0 -1>}
 deform_groin_r {parent=deform_groin_base_r, loc=<0 0 0>, rot=<0 0 0 1>}
 deform_glute_base_l {parent=pelvis, loc=<-7.3466 -9.0213 -9.1245>, rot=<0.0172 0.031 -0.4477 -0.8935>}
 deform_glute_l {parent=deform_glute_base_l, loc=<0 0 0>, rot=<0 0 0 1>}
 deform_glute_base_r {parent=pelvis, loc=<-7.3464 -9.0213 9.1245>, rot=<0.031 -0.0172 0.8935 -0.4477>}
 deform_glute_r {parent=deform_glute_base_r, loc=<0 0 0>, rot=<0 0 0 1>}
 dyn_skirt_bk_l_base {parent=pelvis, loc=<-3.9783 -9.8822 -10.926>, rot=<-0.0661 -0.8687 0.4904 0.0219>}
 dyn_skirt_bk_l_01 {parent=dyn_skirt_bk_l_base, loc=<0.5 0 0>, rot=<0 0 0 -1>}
 dyn_skirt_bk_l_02 {parent=dyn_skirt_bk_l_01, loc=<9 0.0001 0>, rot=<0 0 0 -1>}
 dyn_skirt_bk_centr_base {parent=pelvis, loc=<-4.2963 -12.1245 0>, rot=<-0.0627 -0.7037 0.7053 0.0585>}
 dyn_skirt_bk_centr_01 {parent=dyn_skirt_bk_centr_base, loc=<0.6855 0 0>, rot=<0 0 0 1>}
 dyn_skirt_bk_centr_02 {parent=dyn_skirt_bk_centr_01, loc=<9 0 0.4974>, rot=<0 -0.0381 0 -0.9993>}
 dyn_skirt_side_l_base {parent=pelvis, loc=<1.1705 -1.4758 -13.8486>, rot=<-0.0079 -0.9902 0.027 -0.1365>}
 dyn_skirt_side_l_01 {parent=dyn_skirt_side_l_base, loc=<0.6851 0 0.0003>, rot=<0 0 0 1>}
 dyn_skirt_side_l_02 {parent=dyn_skirt_side_l_01, loc=<8.8709 0 -0.0002>, rot=<0 -0.083 0 -0.9965>}
 dyn_skirt_bk_r_base {parent=pelvis, loc=<-3.9783 -9.8822 10.926>, rot=<-0.0219 -0.4904 0.8687 0.0661>}
 dyn_skirt_bk_r_01 {parent=dyn_skirt_bk_r_base, loc=<0.5 0 0>, rot=<0 0 0 1>}
 dyn_skirt_bk_r_02 {parent=dyn_skirt_bk_r_01, loc=<9 0 0>, rot=<0 0 0 1>}
 dyn_skirt_bk_r_03 {parent=dyn_skirt_bk_r_02, loc=<9 0 0>, rot=<0 0 0 1>}
 dyn_skirt_side_r_base {parent=pelvis, loc=<-2.1227 -1.3578 15.2169>, rot=<0.0891 -0.0266 0.9956 0.0092>}
 dyn_skirt_side_r_01 {parent=dyn_skirt_side_r_base, loc=<0.6855 0 0>, rot=<0 0 0 1>}
 dyn_skirt_side_r_02 {parent=dyn_skirt_side_r_01, loc=<8.8706 0 0>, rot=<0 -0.0723 0 -0.9974>}
 dyn_belt_fr_base_l {parent=pelvis, loc=<9.1572 5.1953 -9.434>, rot=<-0.5266 0.4353 0.7285 -0.0492>}
 dyn_belt_01_l {parent=dyn_belt_fr_base_l, loc=<0.2151 0 0>, rot=<0 0 0 1>}
 dyn_belt_02_l {parent=dyn_belt_01_l, loc=<4.3187 0 0>, rot=<-0.0386 -0.06 -0.1216 -0.99>}
 dyn_belt_03_l {parent=dyn_belt_02_l, loc=<4.4254 0 0>, rot=<-0.0839 -0.3345 -0.2941 -0.8914>}
 dyn_belt_04_l {parent=dyn_belt_03_l, loc=<6.317 0 0>, rot=<0.0959 -0.3785 -0.3157 -0.8648>}
 dyn_belt_05_l {parent=dyn_belt_04_l, loc=<6.7691 0 0>, rot=<0.0592 -0.0997 -0.1099 -0.9872>}
 dyn_belt_06_l {parent=dyn_belt_05_l, loc=<5.1072 0.3659 0.1792>, rot=<0.0743 0 0 -0.9972>}
 dyn_amulet {parent=dyn_belt_03_l, loc=<-0.8848 0.2537 3.5401>, rot=<0.1186 0.7624 0.1571 -0.6164>}
 dyn_maskholder {parent=dyn_amulet, loc=<-0.01 0 -0.03>, rot=<0 0 0 1>}
 dyn_belt_fr_base_r {parent=pelvis, loc=<9.1568 5.1953 9.4344>, rot=<0.5266 -0.4353 0.7285 -0.0492>}
 dyn_belt_01_r {parent=dyn_belt_fr_base_r, loc=<0.2149 -0.0002 0.0005>, rot=<0 0 0 1>}
 dyn_belt_02_r {parent=dyn_belt_01_r, loc=<4.3188 0 0>, rot=<0.0386 0.06 -0.1216 -0.99>}
 dyn_belt_03_r {parent=dyn_belt_02_r, loc=<4.4253 0 0.0001>, rot=<0.0839 0.3345 -0.2941 -0.8914>}
 dyn_belt_04_r {parent=dyn_belt_03_r, loc=<6.317 0.0002 -0.0004>, rot=<-0.0959 0.3785 -0.3157 -0.8648>}
 dyn_belt_05_r {parent=dyn_belt_04_r, loc=<6.7693 -0.0002 0.0004>, rot=<-0.0592 0.0997 -0.1099 -0.9872>}
 dyn_belt_06_r {parent=dyn_belt_05_r, loc=<5.1072 0.3659 -0.1792>, rot=<-0.0743 0 0 -0.9972>}
 dyn_jacket_r_back_base {parent=pelvis, loc=<-0.119 -11.2589 6.7754>, rot=<0.1195 -0.9767 0.1747 0.0353>}
 dyn_jacket_01_r_back {parent=dyn_jacket_r_back_base, loc=<0.4499 0 0>, rot=<0 0 0 1>}
 dyn_jacket_02_r_back {parent=dyn_jacket_01_r_back, loc=<11.8204 -0.0239 -0.6015>, rot=<-0.0009 -0.0254 -0.0361 -0.999>}
 dyn_jacket_03_r_back {parent=dyn_jacket_02_r_back, loc=<11.8368 0 0.0001>, rot=<0 0 -0.008 -1>}
 dyn_jacket_r_fr_base {parent=pelvis, loc=<0.5292 6.3085 11.7516>, rot=<0.0442 -0.3143 0.9457 0.0703>}
 dyn_jacket_01_r_fr {parent=dyn_jacket_r_fr_base, loc=<0.4499 0 0>, rot=<0 0 0 1>}
 dyn_jacket_02_r_fr {parent=dyn_jacket_01_r_fr, loc=<16.1422 -0.2777 0>, rot=<0 0 -0.0114 -0.9999>}
 dyn_jacket_03_r_fr {parent=dyn_jacket_02_r_fr, loc=<16.1421 0.1762 0>, rot=<0 0 0 1>}
 dyn_jacket_r_side_base {parent=pelvis, loc=<-0.0869 -2.263 14.1489>, rot=<0.0959 -0.7605 0.6357 0.091>}
 dyn_jacket_01_r_side {parent=dyn_jacket_r_side_base, loc=<0.4499 0 0>, rot=<0 0 0 1>}
 dyn_jacket_02_r_side {parent=dyn_jacket_01_r_side, loc=<13.5546 0 0>, rot=<0 0 -0.0414 -0.9991>}
 dyn_jacket_03_r_side {parent=dyn_jacket_02_r_side, loc=<13.5546 0 0>, rot=<0 0 0.005 -1>}
 dyn_jacket_l_fr_base {parent=pelvis, loc=<0.5293 6.3085 -11.7517>, rot=<-0.0442 0.3143 0.9457 0.0703>}
 dyn_jacket_01_l_fr {parent=dyn_jacket_l_fr_base, loc=<0.4499 0 0>, rot=<0 0 0 1>}
 dyn_jacket_02_l_fr {parent=dyn_jacket_01_l_fr, loc=<16.1421 -0.2777 0>, rot=<0 0 -0.0114 -0.9999>}
 dyn_jacket_03_l_fr {parent=dyn_jacket_02_l_fr, loc=<16.1421 0.1761 0>, rot=<0 0 0 1>}
 dyn_jacket_l_side_base {parent=pelvis, loc=<-0.0869 -2.263 -14.1489>, rot=<0.0959 -0.7605 -0.6357 -0.091>}
 dyn_jacket_01_l_side {parent=dyn_jacket_l_side_base, loc=<0.4499 0 0>, rot=<0 0 0 1>}
 dyn_jacket_02_l_side {parent=dyn_jacket_01_l_side, loc=<13.5546 0 0>, rot=<0 0 -0.0414 -0.9991>}
 dyn_jacket_03_l_side {parent=dyn_jacket_02_l_side, loc=<13.5546 0 0>, rot=<0 0 0.005 -1>}
 dyn_jacket_l_back_base {parent=pelvis, loc=<-0.1202 -11.267 -6.754>, rot=<0.1151 -0.9755 -0.1777 -0.0601>}
 dyn_jacket_01_l_back {parent=dyn_jacket_l_back_base, loc=<0.4499 0 0>, rot=<0 0 0 1>}
 dyn_jacket_02_l_back {parent=dyn_jacket_01_l_back, loc=<11.8357 -0.0239 0>, rot=<0 0 -0.0361 -0.9993>}
 dyn_jacket_03_l_back {parent=dyn_jacket_02_l_back, loc=<11.8367 0 0>, rot=<0 0 -0.008 -1>}
 dyn_root_floor {parent=pelvis, loc=<-97.9752 4.0424 0>, rot=<-0.0127 -0.707 0.0127 -0.707>}
 dyn_collision_mask {parent=pelvis, loc=<1.802 0.242 -13.4551>, rot=<-0.0127 -0.707 0.0127 -0.707>}
 dyn_collision_strap {parent=pelvis, loc=<5.1199 -0.6215 -11.3337>, rot=<-0.0127 -0.707 0.0127 -0.707>}
]
variables [
 SPR_UP_thigh_l {type=double, default=0}
 SPR_UP_thigh_r {type=double, default=0}
 contrs {type=double, default=0}
 crouch {type=bool, default=False}
 head_vector {type=FVector, default=(X=0,Y=0,Z=0)}
 mask {type=bool, default=False}
 rotate_collar {type=double, default=0}
 target {type=bool, default=False}
 targetAlpha {type=double, default=0}
]
