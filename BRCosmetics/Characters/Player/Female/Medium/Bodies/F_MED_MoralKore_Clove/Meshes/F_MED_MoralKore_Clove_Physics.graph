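# Physics asset for the MoralKore Clove female medium body cosmetic, with capsule bodies for secondary motion and collision.
# Legs, pelvis, spine, root and mask/strap collision bodies accompany limited-angle chains for six jacket panels and two belts hung from the pelvis.
# The amulet and mask holder branch off the left belt.
# Free-angle constraints cross-link neighbouring jacket tips on each side, while back skirt bodies have no constraints here.
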PhysicsAsset /BRCosmetics/Characters/Player/Female/Medium/Bodies/F_MED_MoralKore_Clove/Meshes/F_MED_MoralKore_Clove_Physics
bodies {
 calf_l [{shapeType=Capsule, length=24.4709, radius=9, center=<-18.9452 -1.7087 -1.133>, rotation=<rot 0 94.5172 -87.9734>}]
 calf_r [{shapeType=Capsule, length=24.4709, radius=9, center=<18.9452 1.7087 1.133>, rotation=<rot 0 94.5172 92.0266>}]
 dyn_amulet [
  {shapeType=Capsule, length=15.9363, radius=6, shapeName=dyn_amulet_capsule, rotation=<rot 0 0 16.3688>}
  {shapeType=Capsule, length=19.0679, radius=6, shapeName=dyn_amulet_capsule, center=<10 0 0>, rotation=<rot 0 0 16.3688>}
 ]
 dyn_belt_01_l [{shapeType=Capsule, length=3, radius=3.2, shapeName=dyn_belt_01_l_capsule, center=<3 0 0>, rotation=<rot 90 0 0>}]
 dyn_belt_01_r [{shapeType=Capsule, length=3, radius=3, shapeName=dyn_belt_01_r_capsule, center=<3 0 0>, rotation=<rot 90 0 0>}]
 dyn_belt_02_l [{shapeType=Capsule, length=3, radius=3.2, shapeName=dyn_belt_02_l_capsule, center=<3 0 0>, rotation=<rot 90 0 0>}]
 dyn_belt_02_r [{shapeType=Capsule, length=3, radius=6, shapeName=dyn_belt_02_r_capsule, center=<3 0 0>, rotation=<rot 90 0 0>}]
 dyn_belt_03_l [{shapeType=Capsule, length=3, radius=3, shapeName=dyn_belt_03_l_capsule, center=<3 0 0>, rotation=<rot 90 0 0>}]
 dyn_belt_03_r [{shapeType=Capsule, length=3, radius=4.5, shapeName=dyn_belt_03_r_capsule, center=<3 0 0>, rotation=<rot 90 0 0>}]
 dyn_belt_04_l [{shapeType=Capsule, length=3, radius=2, shapeName=dyn_belt_04_l_capsule, center=<3 0 0>, rotation=<rot 90 0 0>}]
 dyn_belt_04_r [{shapeType=Capsule, length=3, radius=4, shapeName=dyn_belt_04_r_capsule, center=<3 0 0>, rotation=<rot 90 0 0>}]
 dyn_belt_05_l [{shapeType=Capsule, length=3, radius=2, shapeName=dyn_belt_05_l_capsule, center=<3 0 0>, rotation=<rot 90 0 0>}]
 dyn_belt_05_r [{shapeType=Capsule, length=3, radius=4, shapeName=dyn_belt_05_r_capsule, center=<3 0 0>, rotation=<rot 90 0 0>}]
 dyn_belt_06_l [{shapeType=Capsule, length=3, radius=2, shapeName=dyn_belt_06_l_capsule, center=<3 0 0>, rotation=<rot 90 0 0>}]
 dyn_belt_06_r [{shapeType=Capsule, length=3, radius=5, shapeName=dyn_belt_06_r_capsule, center=<5 0 0>, rotation=<rot 90 0 0>}]
 dyn_belt_fr_base_l [{shapeType=Capsule, length=3, radius=3, shapeName=dyn_belt_fr_base_l_capsule, rotation=<rot 90 0 0>}]
 dyn_belt_fr_base_r [{shapeType=Capsule, length=0, radius=0, shapeName=dyn_belt_fr_base_r_capsule, rotation=<rot 90 0 0>}]
 dyn_collision_mask [
  {shapeType=Capsule, length=23.1991, radius=10, shapeName=dyn_collision_mask_capsule, center=<-10.672 -9.7796 0>, rotation=<rot 0 -90 0>}
  {shapeType=Capsule, length=23.1991, radius=10, shapeName=dyn_collision_mask_capsule, center=<-10.672 5.791 0>, rotation=<rot 0 -90 0>}
 ]
 dyn_collision_strap [{
   shapeType=Capsule
   length=22.3535
   radius=6
   shapeName=dyn_collision_strap_capsule
   center=<-6.9934 0.039 1.8939>
   rotation=<rot -3.6681 -115.1854 -12.808>
  }]
 dyn_jacket_01_l_back [{shapeType=Capsule, length=10, radius=3, shapeName=dyn_jacket_01_l_back_capsule, center=<8 0 0>}]
 dyn_jacket_01_l_fr [{shapeType=Capsule, length=10, radius=6, shapeName=dyn_jacket_01_l_fr_capsule, center=<8 0 0>}]
 dyn_jacket_01_l_side [{shapeType=Capsule, length=10, radius=6, shapeName=dyn_jacket_01_l_side_capsule, center=<8 0 0>}]
 dyn_jacket_01_r_back [{shapeType=Capsule, length=10, radius=4, shapeName=dyn_jacket_01_r_back_capsule, center=<8 0 0>}]
 dyn_jacket_01_r_fr [{shapeType=Capsule, length=10, radius=4.5, shapeName=dyn_jacket_01_r_fr_capsule, center=<8 0 0>}]
 dyn_jacket_01_r_side [{shapeType=Capsule, length=10, radius=4, shapeName=dyn_jacket_01_r_side_capsule, center=<8 0 0>}]
 dyn_jacket_02_l_back [{shapeType=Capsule, length=10, radius=4, shapeName=dyn_jacket_02_l_back_capsule, center=<8 0 0>}]
 dyn_jacket_02_l_fr [{shapeType=Capsule, length=10, radius=6, shapeName=dyn_jacket_02_l_fr_capsule, center=<8 0 0>}]
 dyn_jacket_02_l_side [{shapeType=Capsule, length=10, radius=4, shapeName=dyn_jacket_02_l_side_capsule, center=<8 0 0>}]
 dyn_jacket_02_r_back [{shapeType=Capsule, length=10, radius=5, shapeName=dyn_jacket_02_r_back_capsule, center=<8 0 0>}]
 dyn_jacket_02_r_fr [{shapeType=Capsule, length=10, radius=5, shapeName=dyn_jacket_02_r_fr_capsule, center=<8 0 0>}]
 dyn_jacket_02_r_side [{shapeType=Capsule, length=10, radius=4, shapeName=dyn_jacket_02_r_side_capsule, center=<8 0 0>}]
 dyn_jacket_03_l_back [{shapeType=Capsule, length=10, radius=5, shapeName=dyn_jacket_03_l_back_capsule, center=<8 0 0>}]
 dyn_jacket_03_l_fr [{shapeType=Capsule, length=10, radius=5, shapeName=dyn_jacket_03_l_fr_capsule, center=<8 0 0>}]
 dyn_jacket_03_l_side [{
   shapeType=Capsule
   length=10
   radius=5
   shapeName=dyn_jacket_03_l_side_capsule
   center=<7.9683 9.6445 -11.387>
   rotation=<rot 0.0123 0.0022 -59.9086>
  }]
 dyn_jacket_03_r_back [{shapeType=Capsule, length=10, radius=5, shapeName=dyn_jacket_03_r_back_capsule, center=<8 0 0>}]
 dyn_jacket_03_r_fr [{shapeType=Capsule, length=10, radius=3, shapeName=dyn_jacket_03_r_fr_capsule, center=<8 0 0>}]
 dyn_jacket_03_r_side [{shapeType=Capsule, length=10, radius=3, shapeName=dyn_jacket_03_r_side_capsule, center=<8 0 0>}]
 dyn_jacket_l_back_base [{shapeType=Capsule, length=1, radius=2, shapeName=dyn_jacket_l_back_base_capsule}]
 dyn_jacket_l_fr_base [{shapeType=Capsule, length=1, radius=2, shapeName=dyn_jacket_l_fr_base_capsule}]
 dyn_jacket_l_side_base [{shapeType=Capsule, length=1, radius=2, shapeName=dyn_jacket_l_side_base_capsule}]
 dyn_jacket_r_back_base [{shapeType=Capsule, length=1, radius=2, shapeName=dyn_jacket_r_back_base_capsule}]
 dyn_jacket_r_fr_base [{shapeType=Capsule, length=1, radius=2, shapeName=dyn_jacket_r_fr_base_capsule}]
 dyn_jacket_r_side_base [{shapeType=Capsule, length=1, radius=2, shapeName=dyn_jacket_r_side_base_capsule}]
 dyn_maskholder [{
   shapeType=Capsule
   length=3
   radius=9.5
   shapeName=dyn_maskholder_capsule
   center=<1.8259 -0.0019 -0.1634>
   rotation=<rot -0.0901 -90 2.5685>
  }]
 dyn_skirt_bk_centr_01 [{shapeType=Capsule, length=8, radius=2.5, center=<4.7075 0 1.5194>, rotation=<rot 0 0 90>}]
 dyn_skirt_bk_centr_02 [{shapeType=Capsule, length=10, radius=3, center=<5 0 0>, rotation=<rot 0 0 90>}]
 dyn_skirt_bk_centr_base [{shapeType=Capsule, length=1, radius=1}]
 dyn_skirt_bk_l_01 [{shapeType=Capsule, length=8, radius=2.5, center=<4 -2 0>, rotation=<rot 0 0 90>}]
 dyn_skirt_bk_l_02 [{shapeType=Capsule, length=10, radius=3, center=<5 0 0>, rotation=<rot 0 0 90>}]
 dyn_skirt_bk_l_base [{shapeType=Capsule, length=1, radius=1}]
 dyn_skirt_bk_r_01 [{shapeType=Capsule, length=8, radius=2.5, center=<4 2 0>, rotation=<rot 0 0 90>}]
 dyn_skirt_bk_r_02 [{shapeType=Capsule, length=10, radius=2, center=<5 0 0>, rotation=<rot 0 0 90>}]
 dyn_skirt_bk_r_03 [{shapeType=Capsule, length=10, radius=2, center=<4 0 0>, rotation=<rot 0 0 90>}]
 dyn_skirt_bk_r_base [{shapeType=Capsule, length=1, radius=1}]
 foot_l [{shapeType=Capsule, length=21.3904, radius=7.614, center=<-1.9917 7.4307 0.3103>, rotation=<rot 0 3.2295 93.3941>}]
 foot_r [{shapeType=Capsule, length=21.3904, radius=7.614, center=<1.9917 -7.4307 -0.3103>, rotation=<rot 0 3.2295 -86.6059>}]
 pelvis [
  {shapeType=Capsule, length=1.7804, radius=10, center=<0 -2.5111 0>}
  {shapeType=Capsule, length=7.1636, radius=8.3639, shapeName=pelvis_capsule, center=<-8.571 -2.5111 0>}
 ]
 root [{shapeType=Capsule, length=0, radius=29.7671, shapeName=root_capsule, center=<0 0 25.9014>, rotation=<rot 0 -90 0>}]
 spine_01 [{shapeType=Capsule, length=10, radius=6, shapeName=spine_01_capsule, rotation=<rot 0 -90 0>}]
 spine_02 [{shapeType=Capsule, length=10, radius=6, shapeName=spine_02_capsule, center=<2.2574 0.5807 -0.114>, rotation=<rot 0 -177.3741 -179.0048>}]
 thigh_l [{shapeType=Capsule, length=26.6739, radius=6.9677, center=<-22.1982 -0.2153 2.2717>, rotation=<rot 0 -87.6288 -97.9391>}]
 thigh_r [
  {shapeType=Capsule, length=26.6739, radius=6.9677, center=<22.1982 0.2153 -2.2717>, rotation=<rot 0 -87.6288 82.0609>}
  {
   shapeType=Capsule
   length=13.0164
   radius=6.9677
   shapeName=thigh_r_capsule
   center=<27.9571 -0.9641 -1.072>
   rotation=<rot 0 -87.6288 82.0609>
  }
 ]
}
constraints [
 {bone1Name=thigh_l, bone2Name=pelvis, swing1LimitDegrees=45, swing2LimitDegrees=45, twistLimitDegrees=45}
 {bone1Name=calf_l, bone2Name=thigh_l, swing1LimitDegrees=45, swing2LimitDegrees=45, twistLimitDegrees=45}
 {bone1Name=foot_l, bone2Name=calf_l, swing1LimitDegrees=45, swing2LimitDegrees=45, twistLimitDegrees=45}
 {bone1Name=thigh_r, bone2Name=pelvis, swing1LimitDegrees=45, swing2LimitDegrees=45, twistLimitDegrees=45}
 {bone1Name=calf_r, bone2Name=thigh_r, swing1LimitDegrees=45, swing2LimitDegrees=45, twistLimitDegrees=45}
 {bone1Name=foot_r, bone2Name=calf_r, swing1LimitDegrees=45, swing2LimitDegrees=45, twistLimitDegrees=45}
 {bone1Name=dyn_jacket_l_back_base, bone2Name=pelvis, swing1LimitDegrees=10, swing2LimitDegrees=30, twistLimitDegrees=5}
 {bone1Name=dyn_jacket_01_l_back, bone2Name=dyn_jacket_l_back_base, swing1LimitDegrees=10, swing2LimitDegrees=30, twistLimitDegrees=5}
 {bone1Name=dyn_jacket_02_l_back, bone2Name=dyn_jacket_01_l_back, swing1LimitDegrees=10, swing2LimitDegrees=30, twistLimitDegrees=5}
 {bone1Name=dyn_jacket_03_l_back, bone2Name=dyn_jacket_02_l_back, swing1LimitDegrees=10, swing2LimitDegrees=30, twistLimitDegrees=5}
 {bone1Name=dyn_jacket_l_fr_base, bone2Name=pelvis, swing1LimitDegrees=10, swing2LimitDegrees=30, twistLimitDegrees=5}
 {bone1Name=dyn_jacket_01_l_fr, bone2Name=dyn_jacket_l_fr_base, swing1LimitDegrees=10, swing2LimitDegrees=30, twistLimitDegrees=15}
 {bone1Name=dyn_jacket_02_l_fr, bone2Name=dyn_jacket_01_l_fr, swing1LimitDegrees=60, swing2LimitDegrees=60, twistLimitDegrees=15}
 {bone1Name=dyn_jacket_03_l_fr, bone2Name=dyn_jacket_02_l_fr, swing1LimitDegrees=10, swing2LimitDegrees=30, twistLimitDegrees=15}
 {bone1Name=dyn_jacket_l_side_base, bone2Name=pelvis, swing1LimitDegrees=10, swing2LimitDegrees=30, twistLimitDegrees=5}
 {bone1Name=dyn_jacket_01_l_side, bone2Name=dyn_jacket_l_side_base, swing1LimitDegrees=10, swing2LimitDegrees=30, twistLimitDegrees=5}
 {bone1Name=dyn_jacket_02_l_side, bone2Name=dyn_jacket_01_l_side, swing1LimitDegrees=10, swing2LimitDegrees=30, twistLimitDegrees=5}
 {bone1Name=dyn_jacket_03_l_side, bone2Name=dyn_jacket_02_l_side, swing1LimitDegrees=10, swing2LimitDegrees=30, twistLimitDegrees=5}
 {bone1Name=dyn_jacket_r_back_base, bone2Name=pelvis, swing1LimitDegrees=10, swing2LimitDegrees=30, twistLimitDegrees=5}
 {bone1Name=dyn_jacket_01_r_back, bone2Name=dyn_jacket_r_back_base, swing1LimitDegrees=10, swing2LimitDegrees=30, twistLimitDegrees=5}
 {bone1Name=dyn_jacket_02_r_back, bone2Name=dyn_jacket_01_r_back, swing1LimitDegrees=10, swing2LimitDegrees=30, twistLimitDegrees=5}
 {bone1Name=dyn_jacket_03_r_back, bone2Name=dyn_jacket_02_r_back, swing1LimitDegrees=10, swing2LimitDegrees=30, twistLimitDegrees=5}
 {bone1Name=dyn_jacket_r_fr_base, bone2Name=pelvis, swing1LimitDegrees=10, swing2LimitDegrees=30, twistLimitDegrees=5}
 {bone1Name=dyn_jacket_01_r_fr, bone2Name=dyn_jacket_r_fr_base, swing1LimitDegrees=30, swing2LimitDegrees=30, twistLimitDegrees=5}
 {bone1Name=dyn_jacket_02_r_fr, bone2Name=dyn_jacket_01_r_fr, swing1LimitDegrees=45, swing2LimitDegrees=45, twistLimitDegrees=5}
 {bone1Name=dyn_jacket_03_r_fr, bone2Name=dyn_jacket_02_r_fr, swing1LimitDegrees=60, swing2LimitDegrees=60, twistLimitDegrees=5}
 {bone1Name=dyn_jacket_r_side_base, bone2Name=pelvis, swing1LimitDegrees=10, swing2LimitDegrees=30, twistLimitDegrees=5}
 {bone1Name=dyn_jacket_01_r_side, bone2Name=dyn_jacket_r_side_base, swing1LimitDegrees=20, swing2LimitDegrees=30, twistLimitDegrees=5}
 {bone1Name=dyn_jacket_02_r_side, bone2Name=dyn_jacket_01_r_side, swing1LimitDegrees=15, swing2LimitDegrees=15, twistLimitDegrees=5}
 {bone1Name=dyn_jacket_03_r_side, bone2Name=dyn_jacket_02_r_side, swing1LimitDegrees=15, swing2LimitDegrees=15, twistLimitDegrees=5}
 {
  bone1Name=dyn_jacket_03_l_fr
  bone2Name=dyn_jacket_03_l_side
  swing1LimitDegrees=45
  swing1Motion=Free
  swing2LimitDegrees=45
  swing2Motion=Free
  twistLimitDegrees=45
  twistMotion=Free
 }
 {
  bone1Name=dyn_jacket_03_r_side
  bone2Name=dyn_jacket_03_r_back
  swing1LimitDegrees=45
  swing1Motion=Free
  swing2LimitDegrees=45
  swing2Motion=Free
  twistLimitDegrees=45
  twistMotion=Free
 }
 {
  bone1Name=dyn_jacket_03_r_fr
  bone2Name=dyn_jacket_03_r_side
  swing1LimitDegrees=45
  swing1Motion=Free
  swing2LimitDegrees=45
  swing2Motion=Free
  twistLimitDegrees=45
  twistMotion=Free
 }
 {
  bone1Name=pelvis
  bone2Name=root
  swing1LimitDegrees=45
  swing1Motion=Free
  swing2LimitDegrees=45
  swing2Motion=Free
  twistLimitDegrees=45
  twistMotion=Free
 }
 {bone1Name=dyn_belt_06_r, bone2Name=dyn_belt_05_r, swing1LimitDegrees=30, swing2LimitDegrees=30, twistLimitDegrees=5}
 {bone1Name=dyn_belt_fr_base_l, bone2Name=pelvis, swing1LimitDegrees=30, swing2LimitDegrees=30, twistLimitDegrees=5}
 {bone1Name=dyn_belt_01_l, bone2Name=dyn_belt_fr_base_l, swing1LimitDegrees=45, swing2LimitDegrees=45, twistLimitDegrees=5}
 {bone1Name=dyn_belt_02_l, bone2Name=dyn_belt_01_l, swing1LimitDegrees=45, swing2LimitDegrees=45, twistLimitDegrees=5}
 {bone1Name=dyn_belt_03_l, bone2Name=dyn_belt_02_l, swing1LimitDegrees=45, swing2LimitDegrees=45, twistLimitDegrees=5}
 {bone1Name=dyn_amulet, bone2Name=dyn_belt_03_l, swing1LimitDegrees=45, swing2LimitDegrees=45, twistLimitDegrees=5}
 {bone1Name=dyn_belt_04_l, bone2Name=dyn_belt_03_l, swing1LimitDegrees=45, swing2LimitDegrees=45, twistLimitDegrees=5}
 {bone1Name=dyn_belt_05_l, bone2Name=dyn_belt_04_l, swing1LimitDegrees=45, swing2LimitDegrees=45, twistLimitDegrees=5}
 {bone1Name=dyn_belt_06_l, bone2Name=dyn_belt_05_l, swing1LimitDegrees=45, swing2LimitDegrees=45, twistLimitDegrees=5}
 {bone1Name=dyn_belt_fr_base_r, bone2Name=pelvis, swing1LimitDegrees=30, swing2LimitDegrees=30, twistLimitDegrees=5}
 {bone1Name=dyn_belt_01_r, bone2Name=dyn_belt_fr_base_r, swing1LimitDegrees=30, swing2LimitDegrees=30, twistLimitDegrees=5}
 {bone1Name=dyn_belt_02_r, bone2Name=dyn_belt_01_r, swing1LimitDegrees=30, swing2LimitDegrees=30, twistLimitDegrees=5}
 {bone1Name=dyn_belt_03_r, bone2Name=dyn_belt_02_r, swing1LimitDegrees=30, swing2LimitDegrees=30, twistLimitDegrees=5}
 {bone1Name=dyn_belt_04_r, bone2Name=dyn_belt_03_r, swing1LimitDegrees=30, swing2LimitDegrees=30, twistLimitDegrees=5}
 {bone1Name=dyn_belt_05_r, bone2Name=dyn_belt_04_r, swing1LimitDegrees=30, swing2LimitDegrees=30, twistLimitDegrees=5}
 {
  bone1Name=dyn_collision_mask
  bone2Name=pelvis
  swing1LimitDegrees=45
  swing1Motion=Free
  swing2LimitDegrees=45
  swing2Motion=Free
  twistLimitDegrees=45
  twistMotion=Free
 }
 {
  bone1Name=dyn_collision_strap
  bone2Name=pelvis
  swing1LimitDegrees=45
  swing1Motion=Free
  swing2LimitDegrees=45
  swing2Motion=Free
  twistLimitDegrees=45
  twistMotion=Free
 }
 {
  bone1Name=dyn_jacket_03_l_back
  bone2Name=dyn_jacket_03_l_side
  swing1LimitDegrees=45
  swing1Motion=Free
  swing2LimitDegrees=45
  swing2Motion=Free
  twistLimitDegrees=45
  twistMotion=Free
 }
 {bone1Name=dyn_maskholder, bone2Name=dyn_amulet, swing1LimitDegrees=30, swing2LimitDegrees=10, twistLimitDegrees=5}
 {
  bone1Name=spine_01
  bone2Name=pelvis
  swing1LimitDegrees=45
  swing1Motion=Free
  swing2LimitDegrees=45
  swing2Motion=Free
  twistLimitDegrees=45
  twistMotion=Free
 }
 {
  bone1Name=spine_02
  bone2Name=spine_01
  swing1LimitDegrees=45
  swing1Motion=Free
  swing2LimitDegrees=45
  swing2Motion=Free
  twistLimitDegrees=45
  twistMotion=Free
 }
]
