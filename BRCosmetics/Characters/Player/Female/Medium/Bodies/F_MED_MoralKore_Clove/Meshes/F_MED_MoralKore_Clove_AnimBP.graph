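# Cosmetic AnimBlueprint for the Clove (MoralKore) body, driving secondary motion on its skirt, belt, amulet and mask bones.
# The attached parent's copied pose runs through a PreDyn Control Rig, skirt constraints, blended collision-bone offsets and two RigidBody sims (standard and FE physics assets).
# Blended belt and amulet constraint variants then feed a PostDyn Control Rig.
# The update event sets gravity override, mask-variant, sliding and front-end idle flags.
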AnimBlueprint /BRCosmetics/Characters/Player/Female/Medium/Bodies/F_MED_MoralKore_Clove/Meshes/F_MED_MoralKore_Clove_AnimBP
target_skeleton=&/F_MED_MoralKore_Clove_Skeleton
graphs [
 AnimGraph {nodes [
   n1 = "Constraint - dyn_maskholder (P)"(ComponentPose=Local To Component(), Alpha=1, ConstraintWeights_0=1) {
    alphaBoolBlend {blendOption=Linear}
    boneToModify {boneName=dyn_maskholder}
    constraintSetup [{targetBone {boneName=dyn_belt_03_l}}]
    constraintWeights [1]
   }
   Component To Local(ComponentPose=@n1)
   "Transform (Modify) Bone - Bone: dyn_amulet"(Alpha=1, Translation="-2.255,3.530,-8.654", Rotation="-20.294,2.937,-78.045", Scale="1,1,1") {
    rotationMode=BMM_Additive
    rotationSpace=BCS_BoneSpace
    scaleMode=BMM_Additive
    scaleSpace=BCS_BoneSpace
    translationMode=BMM_Additive
    translationSpace=BCS_BoneSpace
    alphaBoolBlend {blendOption=Linear}
    boneToModify {boneName=dyn_amulet}
    rotation=<rot -20.2939 2.937 -78.045>
    translation=<-2.255 3.53 -8.654>
   }
   n2 = "Constraint - dyn_belt_01_l (P)"(ComponentPose=Local To Component(LocalPose=Use cached pose 'post_dyn'()), Alpha=1, ConstraintWeights_0=1) {
    alphaBoolBlend {blendOption=Linear}
    boneToModify {boneName=dyn_belt_01_l}
    constraintSetup [{targetBone {boneName=pelvis}}]
    constraintWeights [1]
   }
   n3 = "Constraint - dyn_belt_02_l (P)"(ComponentPose=@n2, Alpha=1, ConstraintWeights_0=1) {
    alphaBoolBlend {blendOption=Linear}
    boneToModify {boneName=dyn_belt_02_l}
    constraintSetup [{targetBone {boneName=pelvis}}]
    constraintWeights [1]
   }
   n4 = "Constraint - dyn_belt_03_l (P)"(ComponentPose=@n3, Alpha=1, ConstraintWeights_0=1) {
    alphaBoolBlend {blendOption=Linear}
    boneToModify {boneName=dyn_belt_03_l}
    constraintSetup [{targetBone {boneName=pelvis}}]
    constraintWeights [1]
   }
   n5 = "Transform (Modify) Bone - Bone: dyn_belt_03_l"(ComponentPose=@n4, Alpha=1, Translation="5.801,0.049,-1.729", Rotation="-32.781,37.726,-3.471", Scale="1,1,1") {
    rotationMode=BMM_Additive
    rotationSpace=BCS_BoneSpace
    scaleMode=BMM_Additive
    scaleSpace=BCS_BoneSpace
    translationMode=BMM_Additive
    translationSpace=BCS_BoneSpace
    alphaBoolBlend {blendOption=Linear}
    boneToModify {boneName=dyn_belt_03_l}
    rotation=<rot -32.7809 37.726 -3.471>
    translation=<5.801 0.049 -1.729>
   }
   n6 = "Constraint - dyn_belt_04_l (P)"(ComponentPose=@n5, Alpha=1, ConstraintWeights_0=1) {
    alphaBoolBlend {blendOption=Linear}
    boneToModify {boneName=dyn_belt_04_l}
    constraintSetup [{targetBone {boneName=pelvis}}]
    constraintWeights [1]
   }
   n7 = "Constraint - dyn_belt_05_l (PPPPPPP)"!(
    ComponentPose=@n6
    Alpha=1
    ConstraintWeights_0=0.6522
    ConstraintWeights_1:float=0.2186
    ConstraintWeights_2:float=0.0861
    ConstraintWeights_3:float=0.0205
    ConstraintWeights_4:float=0.0205
    ConstraintWeights_5:float=0.0014
    ConstraintWeights_6:float=0.0006
   ) -> (Pose) {
    alphaBoolBlend {blendOption=Linear}
    boneToModify {boneName=dyn_belt_05_l}
    constraintSetup [
     {targetBone {boneName=pelvis}}
     {targetBone {boneName=spine_02}}
     {targetBone {boneName=spine_01}}
     {targetBone {boneName=spine_03}}
     {targetBone {boneName=spine_04}}
     {targetBone {boneName=deform_glute_l}}
     {targetBone {boneName=thigh_twist_01_l}}
    ]
    constraintWeights [1,1,1,1,1,1,1]
   }
   n8 = "Constraint - dyn_belt_06_l (PPPPP)"!(
    ComponentPose=@n7
    Alpha=1
    ConstraintWeights_0=0.4675
    ConstraintWeights_1:float=0.3346
    ConstraintWeights_2:float=0.1151
    ConstraintWeights_3:float=0.0435
    ConstraintWeights_4:float=0.0393
   ) -> (Pose) {
    alphaBoolBlend {blendOption=Linear}
    boneToModify {boneName=dyn_belt_06_l}
    constraintSetup [
     {targetBone {boneName=pelvis}}
     {targetBone {boneName=spine_02}}
     {targetBone {boneName=spine_01}}
     {targetBone {boneName=spine_03}}
     {targetBone {boneName=spine_04}}
    ]
    constraintWeights [1,1,1,1,1]
   }
   n9 = "Constraint - dyn_amulet (P)"(ComponentPose=@n8, Alpha=1, ConstraintWeights_0=1) {
    alphaBoolBlend {blendOption=Linear}
    boneToModify {boneName=dyn_amulet}
    constraintSetup [{targetBone {boneName=dyn_belt_03_l}}]
    constraintWeights [1]
   }
   n10 = "Transform (Modify) Bone - Bone: dyn_amulet"(ComponentPose=@n9, Alpha=1, Translation="0,0,0", Rotation="-0.216,-20.002,0.199", Scale="1,1,1") {
    rotationMode=BMM_Additive
    rotationSpace=BCS_BoneSpace
    scaleMode=BMM_Additive
    scaleSpace=BCS_BoneSpace
    translationMode=BMM_Additive
    translationSpace=BCS_BoneSpace
    alphaBoolBlend {blendOption=Linear}
    boneToModify {boneName=dyn_amulet}
    rotation=<rot -0.216 -20.002 0.199>
   }
   n11 = Component To Local(ComponentPose=@n10)
   n12 = Use cached pose 'post_dyn'()
   n13 = "Constraint - dyn_belt_01_l (P)"(ComponentPose=Local To Component(LocalPose=Use cached pose 'post_dyn'()), Alpha=1, ConstraintWeights_0=1) {
    alphaBoolBlend {blendOption=Linear}
    boneToModify {boneName=dyn_belt_01_l}
    constraintSetup [{targetBone {boneName=pelvis}}]
    constraintWeights [1]
   }
   n14 = "Constraint - dyn_belt_02_l (P)"(ComponentPose=@n13, Alpha=1, ConstraintWeights_0=1) {
    alphaBoolBlend {blendOption=Linear}
    boneToModify {boneName=dyn_belt_02_l}
    constraintSetup [{targetBone {boneName=pelvis}}]
    constraintWeights [1]
   }
   n15 = "Constraint - dyn_belt_03_l (P)"(ComponentPose=@n14, Alpha=1, ConstraintWeights_0=1) {
    alphaBoolBlend {blendOption=Linear}
    boneToModify {boneName=dyn_belt_03_l}
    constraintSetup [{targetBone {boneName=pelvis}}]
    constraintWeights [1]
   }
   n16 = "Constraint - dyn_belt_04_l (P)"(ComponentPose=@n15, Alpha=1, ConstraintWeights_0=1) {
    alphaBoolBlend {blendOption=Linear}
    boneToModify {boneName=dyn_belt_04_l}
    constraintSetup [{targetBone {boneName=pelvis}}]
    constraintWeights [1]
   }
   n17 = "Constraint - dyn_belt_05_l (PPPPPPP)"(
    ComponentPose=@n16
    Alpha=1
    ConstraintWeights_0=0.6522
    ConstraintWeights_1=0.2186
    ConstraintWeights_2=0.0861
    ConstraintWeights_3=0.0205
    ConstraintWeights_4=0.0205
    ConstraintWeights_5=0.0014
    ConstraintWeights_6=0.0006
   ) {
    alphaBoolBlend {blendOption=Linear}
    boneToModify {boneName=dyn_belt_05_l}
    constraintSetup [
     {targetBone {boneName=pelvis}}
     {targetBone {boneName=spine_02}}
     {targetBone {boneName=spine_01}}
     {targetBone {boneName=spine_03}}
     {targetBone {boneName=spine_04}}
     {targetBone {boneName=deform_glute_l}}
     {targetBone {boneName=thigh_twist_01_l}}
    ]
    constraintWeights [1,1,1,1,1,1,1]
   }
   n18 = "Constraint - dyn_belt_06_l (PPPPP)"(
    ComponentPose=@n17
    Alpha=1
    ConstraintWeights_0=0.4675
    ConstraintWeights_1=0.3346
    ConstraintWeights_2=0.1151
    ConstraintWeights_3=0.0435
    ConstraintWeights_4=0.0393
   ) {
    alphaBoolBlend {blendOption=Linear}
    boneToModify {boneName=dyn_belt_06_l}
    constraintSetup [
     {targetBone {boneName=pelvis}}
     {targetBone {boneName=spine_02}}
     {targetBone {boneName=spine_01}}
     {targetBone {boneName=spine_03}}
     {targetBone {boneName=spine_04}}
    ]
    constraintWeights [1,1,1,1,1]
   }
   n19 = "Constraint - dyn_amulet (P)"(ComponentPose=@n18, Alpha=1, ConstraintWeights_0=1) {
    alphaBoolBlend {blendOption=Linear}
    boneToModify {boneName=dyn_amulet}
    constraintSetup [{targetBone {boneName=pelvis}}]
    constraintWeights [1]
   }
   n20 = "Transform (Modify) Bone - Bone: dyn_amulet"(ComponentPose=@n19, Alpha=1, Translation="-0.886,-1.184,-4.664", Rotation="-12.609,5.722,-56.088", Scale="1,1,1") {
    rotationMode=BMM_Additive
    rotationSpace=BCS_BoneSpace
    scaleMode=BMM_Additive
    scaleSpace=BCS_BoneSpace
    translationMode=BMM_Additive
    translationSpace=BCS_BoneSpace
    alphaBoolBlend {blendOption=Linear}
    boneToModify {boneName=dyn_amulet}
    rotation=<rot -12.609 5.722 -56.088>
    translation=<-0.886 -1.184 -4.664>
   }
   n21 = Component To Local(ComponentPose=@n20)
   n22 = Use cached pose 'post_dyn'()
   n23 = Control Rig!(
    Source=Blend Poses by bool(
     BlendPose_0=Blend Poses by bool(BlendPose_0=@n11, BlendPose_1=@n12) {blendType=Linear}
     BlendPose_1=Blend Poses by bool(BlendPose_0=@n21, BlendPose_1=@n22) {blendType=Linear}
    ) {blendType=Linear}
    Alpha
    crouch:bool=false
    mask:bool=false
    sliding:bool=Get sliding() -> (sliding:bool)
    sliding_melee:bool=Get sliding_melee() -> (sliding_melee:bool)
    FE:bool=Get bIsInFrontEnd()
    emote:bool=Get bIsPlayingEmote()
   ) -> (Pose) {
    alphaBoolBlend {blendOption=Linear}
    controlRigAssetReference {blueprintRigClass=&/F_MED_MoralKore_Clove_CtrlRig_PostDyn.F_MED_MoralKore_Clove_CtrlRig_PostDyn_C}
   }
   Output Pose(Result=@n23) {layerGroup=None}
   n24 = Control Rig!(Source=Copy Pose From Mesh() {bUseAttachedParent=true}, Alpha, crouch=false, target:bool=false, mask=false) -> (Pose) {
    alphaBoolBlend {blendOption=Linear}
    controlRigAssetReference {blueprintRigClass=&/F_MED_MoralKore_Clove_CtrlRig_PreDyn.F_MED_MoralKore_Clove_CtrlRig_PreDyn_C}
   }
   n25 = "Constraint - dyn_skirt_bk_r_base (PPP)"!(
    ComponentPose=Local To Component(LocalPose=@n24)
    Alpha=1
    ConstraintWeights_0=0.4016
    ConstraintWeights_1:float=0.3953
    ConstraintWeights_2:float=0.203
   ) -> (Pose) {
    alphaBoolBlend {blendOption=Linear}
    boneToModify {boneName=dyn_skirt_bk_r_base}
    constraintSetup [{targetBone {boneName=pelvis}}, {targetBone {boneName=deform_glute_r}}, {targetBone {boneName=thigh_twist_01_r}}]
    constraintWeights [1,1,1]
   }
   n26 = "Constraint - dyn_skirt_bk_r_base (R)"(ComponentPose=@n25, Alpha=0.5, ConstraintWeights_0=1) {
    alphaBoolBlend {blendOption=Linear}
    boneToModify {boneName=dyn_skirt_bk_r_base}
    constraintSetup [{transformType=Rotation, targetBone {boneName=thigh_r}}]
    constraintWeights [1]
   }
   n27 = "Constraint - dyn_skirt_bk_l_base (PPP)"!(ComponentPose=@n26, Alpha=1, ConstraintWeights_0=0.4016, ConstraintWeights_1:float=0.3953, ConstraintWeights_2:float=0.203) -> (Pose) {
    alphaBoolBlend {blendOption=Linear}
    boneToModify {boneName=dyn_skirt_bk_l_base}
    constraintSetup [{targetBone {boneName=pelvis}}, {targetBone {boneName=deform_glute_l}}, {targetBone {boneName=thigh_twist_01_l}}]
    constraintWeights [1,1,1]
   }
   n28 = "Constraint - dyn_skirt_bk_l_base (R)"(ComponentPose=@n27, Alpha=0.5, ConstraintWeights_0=1) {
    alphaBoolBlend {blendOption=Linear}
    boneToModify {boneName=dyn_skirt_bk_l_base}
    constraintSetup [{transformType=Rotation, targetBone {boneName=thigh_l}}]
    constraintWeights [1]
   }
   n29 = "Constraint - dyn_skirt_bk_centr_base (PP)"!(ComponentPose=@n28, Alpha=1, ConstraintWeights_0=0.5, ConstraintWeights_1:float=0.5) -> (Pose) {
    alphaBoolBlend {blendOption=Linear}
    boneToModify {boneName=dyn_skirt_bk_centr_base}
    constraintSetup [{targetBone {boneName=dyn_skirt_bk_l_base}}, {targetBone {boneName=dyn_skirt_bk_r_base}}]
    constraintWeights [1,1]
   }
   Save cached pose 'pre_dyn'(Pose=Component To Local(ComponentPose=@n29)) {cacheName=pre_dyn}
   n30 = "Constraint - dyn_collision_mask (P)"(ComponentPose=Local To Component(LocalPose=Use cached pose 'pre_dyn'()), Alpha=1, ConstraintWeights_0=1) {
    alphaBoolBlend {blendOption=Linear}
    boneToModify {boneName=dyn_collision_mask}
    constraintSetup [{targetBone {boneName=dyn_jacket_01_l_side}}]
    constraintWeights [1]
   }
   n31 = "Transform (Modify) Bone - Bone: dyn_collision_mask"(ComponentPose=@n30, Alpha=1, Translation="-5.499,5.140,-9.167", Rotation="65.389,-33.278,144.870", Scale="1,1,1") {
    rotationMode=BMM_Additive
    rotationSpace=BCS_BoneSpace
    translationMode=BMM_Additive
    translationSpace=BCS_BoneSpace
    alphaBoolBlend {blendOption=Linear}
    boneToModify {boneName=dyn_collision_mask}
    rotation=<rot 65.389 -33.278 144.87>
    translation=<-5.499 5.14 -9.167>
   }
   n32 = "Transform (Modify) Bone - Bone: dyn_collision_strap"(ComponentPose=@n31, Alpha=1, Translation="-1.484,0.199,6.652", Rotation="3.113,0.874,13.755", Scale="1,1,1") {
    rotationMode=BMM_Additive
    rotationSpace=BCS_BoneSpace
    scaleMode=BMM_Additive
    scaleSpace=BCS_BoneSpace
    translationMode=BMM_Additive
    translationSpace=BCS_BoneSpace
    alphaBoolBlend {blendOption=Linear}
    boneToModify {boneName=dyn_collision_strap}
    rotation=<rot 3.113 0.874 13.755>
    translation=<-1.484 0.199 6.652>
   }
   n33 = "Constraint - dyn_collision_mask (P)"(ComponentPose=Local To Component(LocalPose=Use cached pose 'pre_dyn'()), Alpha=1, ConstraintWeights_0=1) {
    alphaBoolBlend {blendOption=Linear}
    boneToModify {boneName=dyn_collision_mask}
    constraintSetup [{targetBone {boneName=dyn_jacket_01_l_side}}]
    constraintWeights [1]
   }
   n34 = "Transform (Modify) Bone - Bone: dyn_collision_mask"(ComponentPose=@n33, Alpha=1, Translation="-0.258,-0.631,-0.318", Rotation="20.498,18.049,10.567", Scale="1,1,1") {
    rotationMode=BMM_Additive
    rotationSpace=BCS_BoneSpace
    translationMode=BMM_Additive
    translationSpace=BCS_BoneSpace
    alphaBoolBlend {blendOption=Linear}
    boneToModify {boneName=dyn_collision_mask}
    rotation=<rot 20.498 18.049 10.567>
    translation=<-0.258 -0.631 -0.318>
   }
   n35 = "Transform (Modify) Bone - Bone: dyn_collision_strap"(ComponentPose=@n34, Alpha=1, Translation="-0.735,-2.269,5.323", Rotation="9.652,0.007,-0.005", Scale="1,1,1") {
    rotationMode=BMM_Additive
    rotationSpace=BCS_BoneSpace
    scaleMode=BMM_Additive
    scaleSpace=BCS_BoneSpace
    translationMode=BMM_Additive
    translationSpace=BCS_BoneSpace
    alphaBoolBlend {blendOption=Linear}
    boneToModify {boneName=dyn_collision_strap}
    rotation=<rot 9.6519 0.007 -0.0051>
    translation=<-0.735 -2.269 5.323>
   }
   n36 = Blend Poses by bool(BlendPose_0=Component To Local(ComponentPose=@n32), BlendPose_1=Component To Local(ComponentPose=@n35)) {blendType=Linear}
   n37 = "Constraint - dyn_collision_mask (P)"(ComponentPose=Local To Component(LocalPose=Use cached pose 'pre_dyn'()), Alpha=1, ConstraintWeights_0=1) {
    alphaBoolBlend {blendOption=Linear}
    boneToModify {boneName=dyn_collision_mask}
    constraintSetup [{targetBone {boneName=dyn_jacket_01_l_side}}]
    constraintWeights [1]
   }
   n38 = "Transform (Modify) Bone - Bone: dyn_collision_mask"(ComponentPose=@n37, Alpha=1, Translation="-0.258,-0.631,-0.318", Rotation="20.498,18.049,10.567", Scale="1,1,1") {
    rotationMode=BMM_Additive
    rotationSpace=BCS_BoneSpace
    translationMode=BMM_Additive
    translationSpace=BCS_BoneSpace
    alphaBoolBlend {blendOption=Linear}
    boneToModify {boneName=dyn_collision_mask}
    rotation=<rot 20.498 18.049 10.567>
    translation=<-0.258 -0.631 -0.318>
   }
   n39 = "Transform (Modify) Bone - Bone: dyn_collision_strap"(ComponentPose=@n38, Alpha=1, Translation="-0.735,-2.269,5.323", Rotation="9.652,0.007,-0.005", Scale="1,1,1") {
    rotationMode=BMM_Additive
    rotationSpace=BCS_BoneSpace
    scaleMode=BMM_Additive
    scaleSpace=BCS_BoneSpace
    translationMode=BMM_Additive
    translationSpace=BCS_BoneSpace
    alphaBoolBlend {blendOption=Linear}
    boneToModify {boneName=dyn_collision_strap}
    rotation=<rot 9.6519 0.007 -0.0051>
    translation=<-0.735 -2.269 5.323>
   }
   n40 = Component To Local(ComponentPose=@n39)
   n41 = "Constraint - dyn_belt_06_l (PPPPP)"(
    ComponentPose=Local To Component(LocalPose=Blend Poses by bool(BlendPose_0=@n40, BlendPose_1=@n36) {blendType=Linear})
    Alpha=1
    ConstraintWeights_0=0.4675
    ConstraintWeights_1=0.3346
    ConstraintWeights_2=0.1151
    ConstraintWeights_3=0.0435
    ConstraintWeights_4=0.0393
   ) {
    alphaBoolBlend {blendOption=Linear}
    boneToModify {boneName=dyn_belt_06_l}
    constraintSetup [
     {targetBone {boneName=pelvis}}
     {targetBone {boneName=spine_02}}
     {targetBone {boneName=spine_01}}
     {targetBone {boneName=spine_03}}
     {targetBone {boneName=spine_04}}
    ]
    constraintWeights [1,1,1,1,1]
   }
   n42 = RigidBody!(
    ComponentPose=@n41
    bAlphaBoolEnabled:bool
    OverrideWorldGravity:Vector="0,0,-980"
    ExternalForce
    SimSpaceSettings:Sim Space Settings Structure=(WorldAlpha=0,VelocityScaleZ=1,DampingAlpha=1,MaxLinearVelocity=10000,MaxAngularVelocity=10000,MaxLinearAcceleration=10000,MaxAngularAcceleration=10000,ExternalLinearDragV=(X=0,Y=0,Z=0),ExternalLinearVelocity=(X=0,Y=0,Z=0),ExternalAngularVelocity=(X=0,Y=0,Z=0))
    bTransferBoneVelocities:bool=False
   ) -> (Pose) {
    alphaInputType=Bool
    bClampLinearTranslationLimitToRefPose=true
    bFreezeIncomingPoseOnStart=true
    bOverrideWorldGravity=true
    bTransferBoneVelocities=true
    overridePhysicsAsset=&/F_MED_MoralKore_Clove_Physics
    simulationSpace=BaseBoneSpace
    alphaBoolBlend {blendInTime=0.8, blendOption=Linear, blendOutTime=0.8}
    baseBoneRef {boneName=pelvis}
    componentAppliedLinearAccClamp=<3000 3000 3000>
    overrideWorldGravity {z=-980}
    simSpaceSettings {maxAngularAcceleration=3000, maxAngularVelocity=3000, maxLinearAcceleration=3000, maxLinearVelocity=3000, velocityScaleZ=0.1, worldAlpha=0.5}
   }
   n43 = RigidBody(
    ComponentPose=@n42
    OverrideWorldGravity="0,0,-980"
    SimSpaceSettings=(WorldAlpha=0,VelocityScaleZ=1,DampingAlpha=1,MaxLinearVelocity=10000,MaxAngularVelocity=10000,MaxLinearAcceleration=10000,MaxAngularAcceleration=10000,ExternalLinearDragV=(X=0,Y=0,Z=0),ExternalLinearVelocity=(X=0,Y=0,Z=0),ExternalAngularVelocity=(X=0,Y=0,Z=0))
    bTransferBoneVelocities=False
   ) {
    alphaInputType=Bool
    bClampLinearTranslationLimitToRefPose=true
    bFreezeIncomingPoseOnStart=true
    bOverrideWorldGravity=true
    bTransferBoneVelocities=true
    overridePhysicsAsset=&/F_MED_MoralKore_Clove_Physics_FE
    simulationSpace=BaseBoneSpace
    alphaBoolBlend {blendInTime=0.8, blendOption=Linear, blendOutTime=0.8}
    baseBoneRef {boneName=pelvis}
    componentAppliedLinearAccClamp=<3000 3000 3000>
    overrideWorldGravity {z=-980}
    simSpaceSettings {maxAngularAcceleration=3000, maxAngularVelocity=3000, maxLinearAcceleration=3000, maxLinearVelocity=3000, velocityScaleZ=0.1, worldAlpha=0.5}
   }
   Save cached pose 'post_dyn'(Pose=Component To Local(ComponentPose=@n43)) {cacheName=post_dyn}
  ]}
 EventGraph {nodes [
   Event BlueprintUpdateAnimation()
   > n1 = Sequence!(execute) -> (then_0, then_1, then_2:Exec, then_3:Exec)
   > n2 = Gravity Override State Function DEPRECATED(CharacterPartAnimInstance=Self-Reference(), GameTimeInSeconds=GetGameTimeInSeconds(), GravityOverrideParams=Get Gravity Override Params())
   > Set Gravity Override(Gravity Override=@n2)
   n3 = AND Boolean(A=Get bIsInFrontEnd(), B=NOT Boolean(A=Get bIsPlayingEmote()))
   Set FE_Idel(execute=@n1.then_3, FE_Idel:bool=@n3) -> (Output_Get:bool)
   > Set dyn(dyn:bool=NOT Boolean(A=@n3)) -> (Output_Get:bool)
   n4 = Get bIsKneeSliding()
   n5 = Get bIsSliding()
   Set sliding(execute=@n1.then_2, sliding:bool=AND Boolean(A=NOT Boolean(A=@n4), B=@n5)) -> (Output_Get:bool)
   > Set sliding_melee(sliding_melee:bool=AND Boolean(A=@n4, B=@n5)) -> (Output_Get:bool)
   Set mask(execute=@n1.then_1, mask:bool=CharacterItemDefinitionHasAnyMetaTag(Tags=Make Array("[0]"=(TagName="Cosmetics.Variant.Property.Mat1")))) -> (Output_Get:bool)
  ]}
]
variables [
 dyn {type=boolean}
 FE_Idel {type=boolean}
 Gravity Override {type=Vector}
 Gravity Override Params {type=GravityOverrideParamsStruct}
 mask {type=boolean}
 sliding {type=boolean}
 sliding_melee {type=boolean}
]
